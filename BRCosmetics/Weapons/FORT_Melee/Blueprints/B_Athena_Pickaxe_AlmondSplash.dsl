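; Pickaxe weapon Blueprint for the AlmondSplash cosmetic that plays a burst effect when its wielder lands.
; On BeginPlay it casts the instigator to PlayerPawn_Athena and binds to OnPawnLanded.
; Landing calls a server FX event, which calls a client FX event.
; Where cosmetic events can execute, that event pulses the idle VFX component's LandBurst float parameter to 1, then back to 0 after a short delay.
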(event EventBeginPlay
  (|CallParentFunction)
  (bind _asplayer_pawn_athena (Utilities|Casting|CastToPlayerPawn_Athena (Game|GetInstigator)))
  (EventDispatchers|BindEventtoOnPawnLanded _asplayer_pawn_athena (delegate Custom|Event_Landed_FX)))

(event Custom|Event_Landed_FX
  (params (Hit :type Hit%20Result%20Structure%20%28by%20ref%29))
  (CallFunction|EventServerFX))

(event Custom|Event_Server_FX
  (CallFunction|EventClientFX))

(event Custom|Event_Client_FX
  (bind _returnvalue (CallFunction|GetIdleVFXComponent))
  (Networking|CanExecuteCosmeticEvents
    (:True
      (Effects|Components|ParticleSystem|SetFloatParameter _returnvalue "LandBurst" 1.0)
      (Utilities|FlowControl|Delay 0.1)
      (Effects|Components|ParticleSystem|SetFloatParameter _returnvalue "LandBurst" 0.0))))

(event EventTick
  (params (DeltaSeconds :type Float%20%28single-precision%29)))

(fn ConstructionScript
  (|CallParentFunction))
