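; Fortnite reactive pickaxe cosmetic (AnvilAlarmRobin) that powers up when striking certain physical surfaces.
; Those impacts start a timeline driving a Reactivity parameter on swing VFX and mesh materials, or rewind it if already late in playback.
; Reactivity gates extra swing and impact sounds, with swing audio fading on unequip or end play.
; An idle Niagara effect spawns on impact and is destroyed when the timeline finishes; frontend previews use a separate timeline.
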(event EventBeginPlay
  (|CallParentFunction)
  (Utilities|FlowControl|Delay 0.8)
  (if (Reactive|Frontend|IsCosmeticObjectDisplayedasPreview 0)
    (bind _returnvalue (Niagara|SpawnSystemAttached "/BRCosmetics/Weapons/FORT_Melee/Pickaxe_AnvilAlarmRobin/FX/NS_Pickaxe_AnvilAlarmRobin_Idle.NS_Pickaxe_AnvilAlarmRobin_Idle" (Variables|Visual|GetWeaponMesh) "None" "0, 0, 0" "0, 0, 0" "SnapToTarget"))
    (Variables|Default|SetNS%20Idle _returnvalue)
    (bind (_direction _reactivity) (Timeline|TL%20Reactivity%20FE PlayFromStart)
      (:Update
        (bind _output_get (Variables|Default|SetReactivity _reactivity))
        (Effects|Components|ParticleSystem|SetFloatParameter (CallFunction|GetSwingVFXComponent) "Reactivity" _output_get)))))

(event Custom|Trigger%20Reactivity%20FX
  (if (and (Components|Timeline|IsPlaying (Variables|B_Athena_Pickaxe_AnvilAlarmRobin|GetTL%20Reactivity)) (>= (Components|Timeline|GetPlaybackPosition (Variables|B_Athena_Pickaxe_AnvilAlarmRobin|GetTL%20Reactivity)) 0.75))
    (CallFunction|RetriggerReactivity)
    (else
      (bind (_direction _reactivity) (Timeline|TL%20Reactivity PlayFromStart)
        (:Update
          (bind _output_get (Variables|Default|SetReactivity _reactivity))
          (Effects|Components|ParticleSystem|SetFloatParameter (CallFunction|GetSwingVFXComponent) "Reactivity" _output_get)
          (Rendering|Material|SetScalarParameterValueonMaterials (Variables|Visual|GetWeaponMesh) "Reactivity" _output_get))
        (:Finished
          (Components|DestroyComponent (Variables|Default|GetNS%20Idle)))))))

(event Custom|Retrigger%20Reactivity
  (Components|Timeline|SetPlaybackPosition (Variables|B_Athena_Pickaxe_AnvilAlarmRobin|GetTL%20Reactivity) 0.75 false true))

(event FortWeapon|EventOnPlayImpactFX
  (params (HitResult :type Hit%20Result%20Structure%20%28by%20ref%29) (ImpactPhysicalSurface :type EPhysicalSurface%20Enum :default "SurfaceType_Default") (SpawnedPSC :type FXSystem%20Component%20Object%20Reference))
  (bind (_bblockinghit _binitialoverlap _time _distance _location) (Collision|BreakHitResult HitResult))
  (|CallParentFunction HitResult ImpactPhysicalSurface SpawnedPSC)
  (if (Utilities|Select :Index ImpactPhysicalSurface :SurfaceType_Default false :SurfaceType1 false :SurfaceType2 false :SurfaceType3 false :SurfaceType4 false :SurfaceType5 false :SurfaceType6 false :SurfaceType7 false :SurfaceType8 false :SurfaceType9 false :SurfaceType10 false :SurfaceType11 false :SurfaceType12 true :SurfaceType13 false :SurfaceType14 true :SurfaceType15 true :SurfaceType16 true :SurfaceType17 false :SurfaceType18 false :SurfaceType19 false :SurfaceType20 false :SurfaceType21 false :SurfaceType22 false :SurfaceType23 false :SurfaceType24 false :SurfaceType25 false :SurfaceType26 false :SurfaceType27 false :SurfaceType28 false :SurfaceType29 false :SurfaceType30 false :SurfaceType31 false :SurfaceType32 false :SurfaceType33 false :SurfaceType34 false)
    (CallFunction|TriggerReactivityFX)
    (CallFunction|PlayActivatedImpactSound _location)
    (elif (> (Variables|Default|GetReactivity) 0.1)
      (CallFunction|PlayActivatedImpactSound _location)))
  (bind _returnvalue (Niagara|SpawnSystemAttached "/BRCosmetics/Weapons/FORT_Melee/Pickaxe_AnvilAlarmRobin/FX/NS_Pickaxe_AnvilAlarmRobin_Impact.NS_Pickaxe_AnvilAlarmRobin_Impact" SpawnedPSC "None" "0, 0, 0" "0, 0, 0" "SnapToTarget"))
  (if (Reactive|Frontend|IsCosmeticObjectDisplayedasPreview 0)
    (Effects|Components|ParticleSystem|SetBoolParameter _returnvalue "bIsFrontEnd" true)
    (else
      (Utilities|IsValid (Variables|Default|GetNS%20Idle)
        (:Is%20Valid
          _)
        (:Is%20Not%20Valid
          (bind _returnvalue_2 (Niagara|SpawnSystemAttached "/BRCosmetics/Weapons/FORT_Melee/Pickaxe_AnvilAlarmRobin/FX/NS_Pickaxe_AnvilAlarmRobin_Idle.NS_Pickaxe_AnvilAlarmRobin_Idle" (Variables|Visual|GetWeaponMesh) "None" "0, 0, 0" "0, 0, 0" "SnapToTarget"))
          (Variables|Default|SetNS%20Idle _returnvalue_2)))))
  (Components|Activation|Deactivate SpawnedPSC))

(event EventPlayWeaponFireCustomAudio
  (params (bPersistentFire :type Boolean) (bSecondaryFire :type Boolean))
  (|CallParentFunction bPersistentFire bSecondaryFire)
  (if (> (Variables|Default|GetReactivity) 0.25)
    (bind _returnvalue (Audio|SpawnSoundAttached (Variables|Default|GetReactive%20Swing%20Sound) (Variables|Visual|GetWeaponMesh)))
    (Variables|Default|SetReactive%20Audio%20Component _returnvalue)))

(event FortWeapon|EventOnUnEquippedAndDetached
  (|CallParentFunction)
  (bind _reactive_audio_component (Variables|Default|GetReactive%20Audio%20Component)
    (:then
      (Audio|Components|Audio|FadeOut _reactive_audio_component 0.2 0.0 "Linear" :!class /Script/Engine.AudioComponent))
    (:else)))

(event EventEndPlay
  (params (EndPlayReason :type EEndPlayReason%20Enum :default "Destroyed"))
  (|CallParentFunction EndPlayReason)
  (bind _reactive_audio_component (Variables|Default|GetReactive%20Audio%20Component)
    (:then
      (Audio|Components|Audio|FadeOut _reactive_audio_component 0.2 0.0 "Linear" :!class /Script/Engine.AudioComponent))
    (:else)))

(event Custom|Play%20Activated%20Impact%20Sound
  (params (Location :type Vector))
  (Audio|SpawnSoundatLocation (Variables|Default|GetReactive%20Impact%20Sound) Location))

(fn ConstructionScript
  (|CallParentFunction))
