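; Pickaxe weapon Blueprint for the ArmyFlour cosmetic that adds a delayed emissive glow and exhaust smoke effect when fire FX play.
; On weapon fire FX it runs the parent logic, stops any running timeline and sets the mesh's Emissive_Intensity parameter to a base value.
; After a retriggerable delay, a timeline lerps that parameter while a Niagara WingSmoke system spawns attached at the FX_Exhaust socket.
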(event EventBeginPlay)

(event EventTick
  (params (DeltaSeconds :type Float%20%28single-precision%29)))

(event EventOnPlayWeaponFireFX
  (params (bPersistentFire :type Boolean) (bSecondaryFire :type Boolean))
  (bind _weaponmesh (Variables|Visual|GetWeaponMesh))
  (|CallParentFunction bPersistentFire bSecondaryFire)
  (Components|Timeline|Stop (Variables|B_Athena_Pickaxe_ArmyFlour|GetTimeline))
  (Rendering|Material|SetScalarParameterValueonMaterials _weaponmesh "Emissive_Intensity" 15.0)
  (Utilities|FlowControl|RetriggerableDelay 0.8)
  (Utilities|FlowControl|Sequence
    (:then_0
      (bind (_direction _emissive_intensity) (Timeline|Timeline PlayFromStart)
        (:Update
          (Rendering|Material|SetScalarParameterValueonMaterials _weaponmesh "Emissive_Intensity" (Math|Float|Lerp 1.5 3e+01 _emissive_intensity)))))
    (:then_1
      (Niagara|SpawnSystemAttached "/BRCosmetics/Weapons/FORT_Melee/Pickaxe_ArmyFlour/FX/NS_Pickaxe_ArmyFlour_WingSmoke.NS_Pickaxe_ArmyFlour_WingSmoke" _weaponmesh "FX_Exhaust"))))

(fn ConstructionScript
  (|CallParentFunction))
