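; Weapon Blueprint for the AuraCop pickaxe cosmetic, driving main-hand and offhand swing trail VFX.
; Init, instigator-set and equip events set IsOffhand on both swing components, so one Niagara system handles both sides.
; Left and right swings pass the montage section index to both components.
; Second right and left swings activate the offhand trail and toggle UseBaseEnd; the right also deactivates the main trail.
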(event EventOnInitWeaponCosmetics
  (|CallParentFunction)
  (Utilities|FlowControl|Sequence
    (:then_0
      ;[1 1 1] 1x swing NS to handle both sides
      (bind _asniagara_particle_system_component (Utilities|Casting|CastToNiagaraComponent (CallFunction|GetSwingVFXComponent)))
      (Niagara|SetNiagaraVariable(Bool) _asniagara_particle_system_component "IsOffhand" false))
    (:then_1
      (bind _asniagara_particle_system_component_2 (Utilities|Casting|CastToNiagaraComponent (CallFunction|GetSwingOffhandVFXComponent)))
      (Niagara|SetNiagaraVariable(Bool) _asniagara_particle_system_component_2 "IsOffhand" true))))

(event FortWeapon|EventOnInstigatorSet
  (|CallParentFunction)
  (Utilities|FlowControl|Sequence
    (:then_0
      ;[1 1 1] 1x swing NS to handle both sides
      (bind _asniagara_particle_system_component (Utilities|Casting|CastToNiagaraComponent (CallFunction|GetSwingVFXComponent)))
      (Niagara|SetNiagaraVariable(Bool) _asniagara_particle_system_component "IsOffhand" false))
    (:then_1
      (bind _asniagara_particle_system_component_2 (Utilities|Casting|CastToNiagaraComponent (CallFunction|GetSwingOffhandVFXComponent)))
      (Niagara|SetNiagaraVariable(Bool) _asniagara_particle_system_component_2 "IsOffhand" true))))

(event FortWeapon|EventOnEquippedAndAttached
  (|CallParentFunction)
  (Utilities|FlowControl|Sequence
    (:then_0
      ;[1 1 1] 1x swing NS to handle both sides
      (bind _asniagara_particle_system_component (Utilities|Casting|CastToNiagaraComponent (CallFunction|GetSwingVFXComponent)))
      (Niagara|SetNiagaraVariable(Bool) _asniagara_particle_system_component "IsOffhand" false))
    (:then_1
      (bind _asniagara_particle_system_component_2 (Utilities|Casting|CastToNiagaraComponent (CallFunction|GetSwingOffhandVFXComponent)))
      (Niagara|SetNiagaraVariable(Bool) _asniagara_particle_system_component_2 "IsOffhand" true))))

(event EventSwingLeft_Common
  (|CallParentFunction)
  (bind _section_index (Utilities|SetupSwingMontageSectionIndex (Variables|Actor|GetInstigator))
    (:Out
      (for _array_element (Utilities|Array|MakeArray (CallFunction|GetSwingVFXComponent) (CallFunction|GetSwingOffhandVFXComponent))
        (bind _asniagara_particle_system_component (Utilities|Casting|CastToNiagaraComponent _array_element))
        (Niagara|SetNiagaraVariable(Int32) _asniagara_particle_system_component "SwingSectionIndex" _section_index)))))

(event EventSwingRight_Common
  (|CallParentFunction)
  (bind _section_index (Utilities|SetupSwingMontageSectionIndex (Variables|Actor|GetInstigator))
    (:Out
      (for _array_element (Utilities|Array|MakeArray (CallFunction|GetSwingVFXComponent) (CallFunction|GetSwingOffhandVFXComponent))
        (bind _asniagara_particle_system_component (Utilities|Casting|CastToNiagaraComponent _array_element))
        (Niagara|SetNiagaraVariable(Int32) _asniagara_particle_system_component "SwingSectionIndex" _section_index)))))

(event EventSwingRight2
  (|CallParentFunction)
  ;[0.15 0.1146 0.02671 0.5] For this pickaxe, we want the offhand trail to show on what's classed as the 2nd right swing
  (Components|Activation|Deactivate (CallFunction|GetSwingVFXComponent))
  (Components|Activation|Activate (CallFunction|GetSwingOffhandVFXComponent) true)
  (CallFunction|SetEnd true))

(event EventSwingLeft2
  (|CallParentFunction)
  (Components|Activation|Activate (CallFunction|GetSwingOffhandVFXComponent) true)
  (CallFunction|SetEnd false))

(event FortWeapon|EventOnPlayImpactFX
  (params (HitResult :type Hit%20Result%20Structure%20%28by%20ref%29) (ImpactPhysicalSurface :type EPhysicalSurface%20Enum :default "SurfaceType_Default") (SpawnedPSC :type FXSystem%20Component%20Object%20Reference))
  ;[0.7865 0.2723 0 0.7217] Frontend impact fx
  (|CallParentFunction HitResult ImpactPhysicalSurface SpawnedPSC))

(fn SetEnd
  (params (UseBaseAsEnd? :type Boolean))
  (for _array_element (Utilities|Array|MakeArray (CallFunction|GetSwingVFXComponent) (CallFunction|GetSwingOffhandVFXComponent))
    (bind _asniagara_particle_system_component (Utilities|Casting|CastToNiagaraComponent _array_element))
    (Niagara|SetNiagaraVariable(Bool) _asniagara_particle_system_component "UseBaseEnd" UseBaseAsEnd?)))

(fn ConstructionScript
  (|CallParentFunction))
