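; Backspin single-wield ToughPack pickaxe cosmetic Blueprint that tells its swing VFX which direction each swing goes.
; On right and left swing events it runs the parent logic, then sets an integer "SwingSectionIndex" parameter on the swing VFX component, using 1 for right and 0 for left.
; BeginPlay and Tick are empty, and the construction script only calls the parent.
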(event EventBeginPlay)

(event EventTick
  (params (DeltaSeconds :type Float%20%28single-precision%29)))

(event EventSwingRight
  (|CallParentFunction)
  (Effects|Components|ParticleSystem|SetIntParameter (CallFunction|GetSwingVFXComponent) "SwingSectionIndex" 1))

(event EventSwingLeft
  (|CallParentFunction)
  (Effects|Components|ParticleSystem|SetIntParameter (CallFunction|GetSwingVFXComponent) "SwingSectionIndex" 0))

(fn ConstructionScript
  (|CallParentFunction))
