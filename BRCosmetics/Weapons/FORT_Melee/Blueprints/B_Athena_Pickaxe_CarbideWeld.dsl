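; Cosmetic Blueprint for the CarbideWeld pickaxe that matches its Niagara effects to the selected emissive variant.
; The variant's index, from a deprecated cosmetic library lookup, feeds a "Variant" float on the anim trails, impact effects and spawned swing effect.
; Left and right swings pass the montage section index to the trail, and section zero hides the trail and spawns a handle-attached system instead.
; Construction disables idle effect auto-activation.
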(event EventOnInitWeaponCosmetics
  (|CallParentFunction)
  (bind _asniagara_particle_system_component (Utilities|Casting|CastToNiagaraComponent (CallFunction|GetIdleVFXComponent)))
  (bind _asniagara_particle_system_component_2 (Utilities|Casting|CastToNiagaraComponent (CallFunction|GetAnimTrailsPSC)))
  (bind _index (Class|FortCosmeticFunctionLibrary|%28Deprecated%29GetVariantChannelActiveIndexForSlot "(TagName=\"LoadoutSlot.Pickaxe\")" (CallFunction|GetCosmeticData) "(TagName=\"Cosmetics.Variant.Channel.Emissive\")" (Utilities|Array|MakeArray "(TagName=\"Cosmetics.Variant.Property.Emissive1\")" "(TagName=\"Cosmetics.Variant.Property.Emissive4\")" "(TagName=\"Cosmetics.Variant.Property.Emissive2\")" "(TagName=\"Cosmetics.Variant.Property.Emissive3\")")))
  (Niagara|SetNiagaraVariable(Float) _asniagara_particle_system_component_2 "Variant" (Math|Conversions|ToFloat(Integer) _index)))

(event FortWeapon|EventOnPlayImpactFX
  (params (HitResult :type Hit%20Result%20Structure%20%28by%20ref%29) (ImpactPhysicalSurface :type EPhysicalSurface%20Enum :default "SurfaceType_Default") (SpawnedPSC :type FXSystem%20Component%20Object%20Reference))
  (|CallParentFunction HitResult ImpactPhysicalSurface SpawnedPSC)
  (Utilities|IsValid SpawnedPSC
    (:Is%20Valid
      (bind _asniagara_particle_system_component (Utilities|Casting|CastToNiagaraComponent SpawnedPSC))
      (bind _index (Class|FortCosmeticFunctionLibrary|%28Deprecated%29GetVariantChannelActiveIndexForSlot "(TagName=\"LoadoutSlot.Pickaxe\")" (CallFunction|GetCosmeticData) "(TagName=\"Cosmetics.Variant.Channel.Emissive\")" (Utilities|Array|MakeArray "(TagName=\"Cosmetics.Variant.Property.Emissive1\")" "(TagName=\"Cosmetics.Variant.Property.Emissive4\")" "(TagName=\"Cosmetics.Variant.Property.Emissive2\")" "(TagName=\"Cosmetics.Variant.Property.Emissive3\")")))
      (Niagara|SetNiagaraVariable(Float) _asniagara_particle_system_component "Variant" (Math|Conversions|ToFloat(Integer) _index)))))

(event EventSwingLeft
  (bind _returnvalue (CallFunction|GetAnimTrailVFXComponent))
  (|CallParentFunction)
  (bind _section_index (Utilities|SetupSwingMontageSectionIndex (Variables|Actor|GetInstigator))
    (:Out
      (Effects|Components|ParticleSystem|SetIntParameter _returnvalue "SwingSectionIndex" _section_index)
      (Development|SetHiddeninGame (CallFunction|GetAnimTrailsPSC) (== _section_index 0) false)
      (if (== _section_index 0)
        (bind _returnvalue_2 (Niagara|SpawnSystemAttached (Variables|Default|GetSystem%20Template) (Variables|Visual|GetWeaponMesh) "Melee_TwoHanded_Handle" "0, 0, 0" "0, 0, 0" "KeepRelativeOffset" true))
        (bind _index (Class|FortCosmeticFunctionLibrary|%28Deprecated%29GetVariantChannelActiveIndexForSlot "(TagName=\"LoadoutSlot.Pickaxe\")" (CallFunction|GetCosmeticData) "(TagName=\"Cosmetics.Variant.Channel.Emissive\")" (Utilities|Array|MakeArray "(TagName=\"Cosmetics.Variant.Property.Emissive1\")" "(TagName=\"Cosmetics.Variant.Property.Emissive4\")" "(TagName=\"Cosmetics.Variant.Property.Emissive2\")" "(TagName=\"Cosmetics.Variant.Property.Emissive3\")")))
        (Niagara|SetNiagaraVariable(Float) _returnvalue_2 "Variant" (Math|Conversions|ToFloat(Integer) _index))))
    (:Cast%20Failed%20Out
      (Effects|Components|ParticleSystem|SetIntParameter _returnvalue "SwingSectionIndex" -1))))

(event EventSwingRight
  (bind _returnvalue (CallFunction|GetAnimTrailVFXComponent))
  (|CallParentFunction)
  (bind _section_index (Utilities|SetupSwingMontageSectionIndex (Variables|Actor|GetInstigator))
    (:Out
      (Effects|Components|ParticleSystem|SetIntParameter _returnvalue "SwingSectionIndex" _section_index)
      (Development|SetHiddeninGame (CallFunction|GetAnimTrailsPSC) (== _section_index 0) false)
      (if (== _section_index 0)
        (bind _returnvalue_2 (Niagara|SpawnSystemAttached (Variables|Default|GetSystem%20Template) (Variables|Visual|GetWeaponMesh) "Melee_TwoHanded_Handle" "0, 0, 0" "0, 0, 0" "KeepRelativeOffset" true))
        (bind _index (Class|FortCosmeticFunctionLibrary|%28Deprecated%29GetVariantChannelActiveIndexForSlot "(TagName=\"LoadoutSlot.Pickaxe\")" (CallFunction|GetCosmeticData) "(TagName=\"Cosmetics.Variant.Channel.Emissive\")" (Utilities|Array|MakeArray "(TagName=\"Cosmetics.Variant.Property.Emissive1\")" "(TagName=\"Cosmetics.Variant.Property.Emissive4\")" "(TagName=\"Cosmetics.Variant.Property.Emissive2\")" "(TagName=\"Cosmetics.Variant.Property.Emissive3\")")))
        (Niagara|SetNiagaraVariable(Float) _returnvalue_2 "Variant" (Math|Conversions|ToFloat(Integer) _index))))
    (:Cast%20Failed%20Out
      (Effects|Components|ParticleSystem|SetIntParameter _returnvalue "SwingSectionIndex" -1))))

(fn ConstructionScript
  (|CallParentFunction)
  (Components|Activation|SetAutoActivate (CallFunction|GetIdlePSC) false))
