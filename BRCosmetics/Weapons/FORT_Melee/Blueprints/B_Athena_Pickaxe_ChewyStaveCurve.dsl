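; Cosmetic pickaxe Blueprint, named for the Chewy Stave, that makes its idle VFX react to melee impacts.
; On equip and attach, it waits briefly, then activates the idle VFX component.
; On impact, only when Swing Right is set, it restarts a timeline whose curve drives the idle effect's ReactiveCurve float parameter.
; Both events and the construction script call their parent implementations first.
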(event FortWeapon|EventOnPlayImpactFX
  (params (HitResult :type Hit%20Result%20Structure%20%28by%20ref%29) (ImpactPhysicalSurface :type EPhysicalSurface%20Enum :default "SurfaceType_Default") (SpawnedPSC :type FXSystem%20Component%20Object%20Reference))
  (|CallParentFunction HitResult ImpactPhysicalSurface SpawnedPSC)
  (if (Variables|Default|GetSwing_Right?)
    (bind (_direction _react) (Timeline|Timeline PlayFromStart)
      (:Update
        (Effects|Components|ParticleSystem|SetFloatParameter (CallFunction|GetIdleVFXComponent) "ReactiveCurve" _react)))))

(event FortWeapon|EventOnEquippedAndAttached
  (|CallParentFunction)
  (Utilities|FlowControl|Delay 0.2)
  (Components|Activation|Activate (CallFunction|GetIdleVFXComponent) false))

(fn ConstructionScript
  (|CallParentFunction))
