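; Cosmetic Blueprint for the ClaimReflect pickaxe that drives its deploy animation and swing trail effects.
; After a short delay on cosmetic init, or on becoming visible, it stamps game time into the blade's dynamic material and idle VFX for a deploy wipe, also printing the material name.
; Each swing passes the montage section index to the trail VFX, or -1 if the cast fails.
; Construction disables idle particle auto-activation.
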(event EventBeginPlay
  (|CallParentFunction))

(event EventOnInitWeaponCosmetics
  (|CallParentFunction)
  (Utilities|FlowControl|Delay 0.2)
  (CallFunction|DeployAnimation))

(event EventOnWeaponVisibilityChanged
  (params (bVisible :type Boolean) (bSetForLocalControllerOnly :type Boolean))
  (|CallParentFunction bVisible bSetForLocalControllerOnly)
  (if bVisible
    (CallFunction|DeployAnimation)))

(event Custom|DeployAnimation
  (bind _weaponmesh (Variables|Visual|GetWeaponMesh))
  (bind _material_instance_dynamic (Class|BTechArtGlobalLibrary|GetDynamicMaterialInstance _weaponmesh))
  (bind _returnvalue (Utilities|Time|GetGameTimeinSeconds))
  (Utilities|FlowControl|Sequence
    (:then_0
      ; Pass timestamp to the blade material to play the deploy wipe animation
      (Rendering|Material|SetScalarParameterValue _material_instance_dynamic "DeployTimestamp" _returnvalue :!class /Script/Engine.MaterialInstanceDynamic)
      (Development|PrintString (Utilities|GetDisplayName _material_instance_dynamic)))
    (:then_1
      (Effects|Components|ParticleSystem|SetFloatParameter (CallFunction|GetIdleVFXComponent) "DeployTimestamp" _returnvalue))))

(event EventSwingRight
  (bind _returnvalue (CallFunction|GetAnimTrailVFXComponent))
  (|CallParentFunction)
  (bind _section_index (Utilities|SetupSwingMontageSectionIndex (Variables|Actor|GetInstigator))
    (:Out
      (Effects|Components|ParticleSystem|SetIntParameter _returnvalue "SwingSectionIndex" _section_index))
    (:Cast%20Failed%20Out
      (Effects|Components|ParticleSystem|SetIntParameter _returnvalue "SwingSectionIndex" -1))))

(event EventSwingLeft
  (bind _returnvalue (CallFunction|GetAnimTrailVFXComponent))
  (|CallParentFunction)
  (bind _section_index (Utilities|SetupSwingMontageSectionIndex (Variables|Actor|GetInstigator))
    (:Out
      (Effects|Components|ParticleSystem|SetIntParameter _returnvalue "SwingSectionIndex" _section_index))
    (:Cast%20Failed%20Out
      (Effects|Components|ParticleSystem|SetIntParameter _returnvalue "SwingSectionIndex" -1))))

(fn ConstructionScript
  (|CallParentFunction)
  (Components|Activation|SetAutoActivate (CallFunction|GetIdlePSC) false))
