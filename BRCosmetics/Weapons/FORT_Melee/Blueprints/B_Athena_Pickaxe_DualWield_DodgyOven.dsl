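; Dual-wield melee pickaxe weapon Blueprint for the DodgyOven cosmetic, with no custom gameplay logic in its graphs.
; The impact FX event and construction script only forward to the parent, while BeginPlay and Tick are disabled placeholders and the tagged-action event is empty.
; The remaining functions return fixed values: no holster, no sound indicator, Friendly affiliation and empty audio parameters.
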(event EventBeginPlay
  (|CallParentFunction) ; This node is disabled and will not be called.\nDrag off pins to build functionality.
)

(event EventTick
  (params (DeltaSeconds :type Float%20%28single-precision%29))
  (|CallParentFunction DeltaSeconds) ; This node is disabled and will not be called.\nDrag off pins to build functionality.
)

(event EventRespondToTaggedAction
  (params (ActionTag :type Gameplay%20Tag%20Structure%20%28by%20ref%29) (bIsApplyingAction :type Boolean)))

(event FortWeapon|EventOnPlayImpactFX
  (params (HitResult :type Hit%20Result%20Structure%20%28by%20ref%29) (ImpactPhysicalSurface :type EPhysicalSurface%20Enum :default "SurfaceType_Default") (SpawnedPSC :type FXSystem%20Component%20Object%20Reference))
  (|CallParentFunction HitResult ImpactPhysicalSurface SpawnedPSC))

(fn CanApplyHolster
  (params (HolsterId :type Name%20%28by%20ref%29 :default "None"))
  (return false))

(fn GetActorAudioParameters
  (return 0))

(fn GetAffiliationsToShowFor
  (return "Friendly"))

(fn ShouldShowSoundIndicator
  (params (PlayerController :type Fort%20Player%20Controller%20Object%20Reference))
  (return false))

(fn ConstructionScript
  (|CallParentFunction))
