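; Dual katana pickaxe weapon Blueprint that drives swing trail and impact VFX for the melee harvesting tool.
; Each left or right swing event passes the montage section index to the trail effect.
; Impacts set an effect number from the current montage section and branch on physical surface.
; Weakspot effects orient to a camera line-trace normal and relay through server-to-client events, while human-hit setup derives yaw from the camera-to-weapon vector, avoiding unreliable close-range traces.
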(event EventSwingLeft
  (bind _returnvalue (CallFunction|GetAnimTrailVFXComponent))
  ; SwingSectionIndex for the Trail FX
  (|CallParentFunction)
  (bind _section_index (Utilities|SetupSwingMontageSectionIndex (Variables|Actor|GetInstigator))
    (:Out
      (Effects|Components|ParticleSystem|SetIntParameter _returnvalue "SwingSectionIndex" _section_index))
    (:Cast%20Failed%20Out
      (Effects|Components|ParticleSystem|SetIntParameter _returnvalue "SwingSectionIndex" -1))))

(event EventSwingLeft2
  (bind _returnvalue (CallFunction|GetAnimTrailVFXComponent))
  ; SwingSectionIndex for the Trail FX
  (|CallParentFunction)
  (bind _section_index (Utilities|SetupSwingMontageSectionIndex (Variables|Actor|GetInstigator))
    (:Out
      (Effects|Components|ParticleSystem|SetIntParameter _returnvalue "SwingSectionIndex" _section_index))
    (:Cast%20Failed%20Out
      (Effects|Components|ParticleSystem|SetIntParameter _returnvalue "SwingSectionIndex" -1))))

(event EventSwingRight
  (bind _returnvalue (CallFunction|GetAnimTrailVFXComponent))
  ; SwingSectionIndex for the Trail FX
  (|CallParentFunction)
  (bind _section_index (Utilities|SetupSwingMontageSectionIndex (Variables|Actor|GetInstigator))
    (:Out
      (Effects|Components|ParticleSystem|SetIntParameter _returnvalue "SwingSectionIndex" _section_index))
    (:Cast%20Failed%20Out
      (Effects|Components|ParticleSystem|SetIntParameter _returnvalue "SwingSectionIndex" -1))))

(event EventSwingRight2
  (bind _returnvalue (CallFunction|GetAnimTrailVFXComponent))
  ; SwingSectionIndex for the Trail FX
  (|CallParentFunction)
  (bind _section_index (Utilities|SetupSwingMontageSectionIndex (Variables|Actor|GetInstigator))
    (:Out
      (Effects|Components|ParticleSystem|SetIntParameter _returnvalue "SwingSectionIndex" _section_index))
    (:Cast%20Failed%20Out
      (Effects|Components|ParticleSystem|SetIntParameter _returnvalue "SwingSectionIndex" -1))))

(event Custom|WeakpointEvent
  (CallFunction|TweakWeakspotFX))

(event Custom|FX_Server
  (CallFunction|FXClient))

(event Custom|FX_Client
  (CallFunction|TweakWeakspotFX))

(event FortWeapon|EventOnPlayImpactFX
  (params (HitResult :type Hit%20Result%20Structure%20%28by%20ref%29) (ImpactPhysicalSurface :type EPhysicalSurface%20Enum :default "SurfaceType_Default") (SpawnedPSC :type FXSystem%20Component%20Object%20Reference))
  (|CallParentFunction HitResult ImpactPhysicalSurface SpawnedPSC)
  (Utilities|FlowControl|Sequence
    (:then_0
      (bind _asfort_pawn (Utilities|Casting|CastToFortPawn (Game|GetInstigator)))
      (Effects|Components|ParticleSystem|SetIntParameter (Variables|Default|GetImpact%20FX) "ImpactNumber" (Utilities|Array|FindItem (Variables|Default|GetMontage_Swing_Sections) (Animation|Montage|MontageGetCurrentSection (Components|SkeletalMesh|GetAnimInstance (Variables|/Script/Engine.Character|GetMesh _asfort_pawn)) (Variables|Default|GetMontageReference)))))
    (:then_1
      (switch Utilities|FlowControl|Switch|SwitchonEPhysicalSurface ImpactPhysicalSurface
        (:SurfaceType4
          _)
        (:SurfaceType5
          _)
        (:SurfaceType12
          ;[1 1 1] Rotation and Position Tweaks for Weakspot FX
          (CallFunction|TweakWeakspotFX :!label _label)
          (CallFunction|FXServer))
        (:SurfaceType14
          (goto _label))
        (:SurfaceType15
          (goto _label))
        (:SurfaceType16
          (goto _label)))
      (if (not (Reactive|Frontend|IsCosmeticObjectDisplayedinFrontendContext 0))
        (Utilities|FlowControl|Delay 0.05)
        (CallFunction|SetupHumanHit)))))

(fn SetupHumanHit
  (bind (_returnvalue_x _returnvalue_y) (Math|Vector|BreakVector (* (- (Transformation|GetActorLocation) (Camera|GetCameraLocation (Game|GetPlayerCameraManager 0))) 4e+02)))
  (Transformation|SetWorldRotation (Variables|Default|GetImpact%20FX) (Math|Rotator|CombineRotators (Math|Conversions|RotationFromXVector (Math|Vector|MakeVector _returnvalue_x _returnvalue_y)) "0.000000,180.000000,0.000000") false false) ; wee-bit cursed, but line trace was failing if the player was in a weird angle or really close ;/
)

(fn TweakWeakspotFX
  (bind _impact_fx (Variables|Default|GetImpact%20FX))
  (bind _outhit (Collision|LineTraceByChannel (Camera|GetCameraLocation (Game|GetPlayerCameraManager 0)) (+ (* (Math|Vector|GetForwardVector (Camera|GetCameraRotation (Game|GetPlayerCameraManager 0))) 4e+03) (Camera|GetCameraLocation (Game|GetPlayerCameraManager 0))) "TraceTypeQuery10"))
  (bind (_outhit_bblockinghit _outhit_binitialoverlap _outhit_time _outhit_distance _outhit_location _outhit_impactpoint _outhit_normal) (Collision|BreakHitResult _outhit))
  (bind (_returnvalue_x _returnvalue_y) (Math|Vector|BreakVector (* _outhit_normal 4e+02)))
  (Transformation|SetWorldRotation _impact_fx (Math|Conversions|RotationFromXVector (Math|Vector|MakeVector _returnvalue_x _returnvalue_y)) false false)
  (Transformation|AddLocalOffset _impact_fx "-20.000000,0.000000,0.000000" false false)
  (Effects|Components|ParticleSystem|SetBoolParameter _impact_fx "WeakpointHit" true))

(fn ConstructionScript
  (|CallParentFunction))
