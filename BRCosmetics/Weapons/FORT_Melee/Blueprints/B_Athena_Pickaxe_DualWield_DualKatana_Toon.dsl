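; Blueprint for a toon-styled dual-katana, dual-wield pickaxe cosmetic weapon in Battle Royale.
; On BeginPlay it calls the parent implementation, then runs the ToonFrontendLUTReplacePickaxe utility, which by its name swaps the toon look's LUT for frontend display.
; The construction script only calls the parent, so all other behavior comes from the parent class.
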(event EventBeginPlay
  (|CallParentFunction)
  (Utilities|ToonFrontendLUTReplacePickaxe))

(fn ConstructionScript
  (|CallParentFunction))
