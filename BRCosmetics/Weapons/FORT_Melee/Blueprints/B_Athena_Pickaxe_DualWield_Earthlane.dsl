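; Blueprint for the Earthlane dual-wield pickaxe cosmetic, adding swing-reactive behavior to its idle VFX.
; The right and left common swing events call their parent, then set the "bisSwinging" bool parameter to true on the idle VFX particle component.
; Nothing here sets the flag back to false, and the construction script only calls its parent.
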(event EventSwingRight_Common
  (|CallParentFunction)
  (Effects|Components|ParticleSystem|SetBoolParameter (CallFunction|GetIdleVFXComponent) "bisSwinging" true))

(event EventSwingLeft_Common
  (|CallParentFunction)
  (Effects|Components|ParticleSystem|SetBoolParameter (CallFunction|GetIdleVFXComponent) "bisSwinging" true))

(fn ConstructionScript
  (|CallParentFunction))
