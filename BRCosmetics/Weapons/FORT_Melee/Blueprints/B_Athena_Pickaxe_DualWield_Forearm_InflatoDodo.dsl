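; Blueprint for the InflatoDodo dual-wield forearm pickaxe, customizing its impact effects for frontend cosmetic previews.
; On equip, it caches the player pawn and the active particle variant index.
; In preview, impacts deactivate the default effect and spawn an inactive Niagara impact relative to the pawn.
; A DoOnce guard, with a short delay against duplicate signals, sets swing section from a cycling counter, preview flag and variant, then activates it.
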(event EventBeginPlay
  (|CallParentFunction) ; This node is disabled and will not be called.\nDrag off pins to build functionality.
)

(event EventTick
  (params (DeltaSeconds :type Float%20%28single-precision%29))
  (|CallParentFunction DeltaSeconds) ; This node is disabled and will not be called.\nDrag off pins to build functionality.
)

(event EventRespondToTaggedAction
  (params (ActionTag :type Gameplay%20Tag%20Structure%20%28by%20ref%29) (bIsApplyingAction :type Boolean)))

(event FortWeapon|EventOnPlayImpactFX
  (params (HitResult :type Hit%20Result%20Structure%20%28by%20ref%29) (ImpactPhysicalSurface :type EPhysicalSurface%20Enum :default "SurfaceType_Default") (SpawnedPSC :type FXSystem%20Component%20Object%20Reference))
  (bind (_returnvalue_location _returnvalue_rotation _returnvalue_scale) (Math|Transform|BreakTransform (Transformation|GetActorTransform (Variables|Default|GetAs%20Fort%20Player%20Pawn))))
  (|CallParentFunction HitResult ImpactPhysicalSurface SpawnedPSC)
  (if (Reactive|Frontend|IsCosmeticObjectDisplayedasPreview 0)
    ;[0 0 0] frontend
    (Components|Activation|Deactivate SpawnedPSC)
    (bind _returnvalue (Niagara|SpawnSystematLocation "/BRCosmetics/Weapons/FORT_Melee/Pickaxe_InflatoDodo/FX/NS_Pickaxe_InflatoDodo_Impact.NS_Pickaxe_InflatoDodo_Impact" (Math|Transform|TransformLocation (Transformation|GetActorTransform (Variables|Default|GetAs%20Fort%20Player%20Pawn)) "100.000000,0.000000,50.000000") (Math|Rotator|CombineRotators _returnvalue_rotation "0.000000,180.000000,0.000000") _returnvalue_scale true false))
    (Utilities|FlowControl|DoOnce false
      (:Completed
        (Effects|Components|ParticleSystem|SetIntParameter _returnvalue "SwingSectionIndex" (Math|Integer|%25%28Integer%29 (Variables|Default|GetFE_Counter) 4))
        (Effects|Components|ParticleSystem|SetBoolParameter _returnvalue "bIsFrontEndPreview" true)
        (Effects|Components|ParticleSystem|SetFloatParameter _returnvalue "Variant" (Math|Conversions|ToFloat(Integer) (Variables|Default|GetFE%20Variant%20Index)))
        (Components|Activation|Activate _returnvalue false)
        (Math|Integer|IncrementInt (Variables|Default|GetFE_Counter)
          (:%20%20
            (Utilities|FlowControl|RetriggerableDelay 0.25) ; prevents duplicated signal
          ))))))

(event FortWeapon|EventOnEquippedAndAttached
  (|CallParentFunction)
  (bind _asfort_player_pawn (Utilities|Casting|CastToFortPlayerPawn (Variables|Actor|GetInstigator)))
  (bind _output_get (Variables|Default|SetAs%20Fort%20Player%20Pawn _asfort_player_pawn))
  (bind _index (CosmeticLoadout|GetVariantChannelActiveIndexforSlotActor "(TagName=\"LoadoutSlot.Pickaxe\")" _output_get "(TagName=\"Cosmetics.Variant.Channel.Particle\")" (Utilities|Array|MakeArray "(TagName=\"Cosmetics.Variant.Property.Particle1\")" "(TagName=\"Cosmetics.Variant.Property.Particle2\")")))
  (Variables|Default|SetFE%20Variant%20Index _index))

(fn CanApplyHolster
  (params (HolsterId :type Name%20%28by%20ref%29 :default "None"))
  (return false))

(fn GetActorAudioParameters
  (return 0))

(fn GetAffiliationsToShowFor
  (return "Friendly"))

(fn ShouldShowSoundIndicator
  (params (PlayerController :type Fort%20Player%20Controller%20Object%20Reference))
  (return false))

(fn ConstructionScript
  (|CallParentFunction))
