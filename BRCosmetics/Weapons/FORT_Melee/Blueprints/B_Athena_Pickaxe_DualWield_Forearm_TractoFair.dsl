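; Dual-wield forearm pickaxe Blueprint (TractoFair) that makes its materials and effects react to weak point hits.
; On impact, it stores the spawned effect and checks the surface against a weak point surface list.
; A weak point hit stamps game time into both weapon meshes and swing and idle effects, flags the impact effect, and restarts a timeline.
; That timeline drives WeakPointActive, so impacts while it equals one also show reactivity.
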(event FortWeapon|EventOnPlayImpactFX
  (params (HitResult :type Hit%20Result%20Structure%20%28by%20ref%29) (ImpactPhysicalSurface :type EPhysicalSurface%20Enum :default "SurfaceType_Default") (SpawnedPSC :type FXSystem%20Component%20Object%20Reference))
  (Variables|Default|SetImpact%20VFX SpawnedPSC)
  (|CallParentFunction HitResult ImpactPhysicalSurface SpawnedPSC)
  (Utilities|FlowControl|Sequence
    (:then_0
      (if (== (Variables|Default|GetWeakPointActive) 1.0)
        (Effects|Components|ParticleSystem|SetFloatParameter (Variables|Default|GetImpact%20VFX) "ReactivityTime" 1.0)))
    (:then_1
      (bind _output_get (Variables|Default|SetCurrent%20Impact%20Physical%20Surface%200 ImpactPhysicalSurface))
      (Variables|Default|SetIs%20Hit%20WeakPoint (Utilities|Array|ContainsItem (Variables|Default|GetWeakPoint%20Physical%20Surface%20Array_0) _output_get))
      (if (Variables|Default|GetIs%20Hit%20WeakPoint)
        (Effects|Components|ParticleSystem|SetFloatParameter (Variables|Default|GetImpact%20VFX) "ReactivityTime" 1.0)
        (CallFunction|OnWeakPointHit true)))))

(event Custom|OnWeakPointHit
  (params (RestartTimeline :type Boolean))
  (CallFunction|MaterialUpdate RestartTimeline))

(event Custom|MaterialUpdate
  (params (RestartTimeline :type Boolean))
  (bind _returnvalue (Utilities|Time|GetGameTimeinSeconds))
  (bind _returnvalue_2 (Utilities|Time|GetGameTimeinSeconds))
  (Effects|Components|ParticleSystem|SetFloatParameter (Variables|Default|GetImpact%20VFX) "ReactivityTime" 1.0)
  (Rendering|Material|SetScalarParameterValueonMaterials (Variables|Visual|GetWeaponMesh) "Reactivity_Time" _returnvalue_2)
  (Rendering|Material|SetScalarParameterValueonMaterials (Variables|Visual|GetWeaponMeshOffhand) "Reactivity_Time" _returnvalue_2)
  (Effects|Components|ParticleSystem|SetFloatParameter (CallFunction|GetSwingVFXComponent) "ReactivityTime" _returnvalue)
  (Effects|Components|ParticleSystem|SetFloatParameter (CallFunction|GetIdleVFXComponent) "ReactivityTime" _returnvalue)
  (if RestartTimeline
    (bind (_direction _weakpointactive) (Timeline|Timeline PlayFromStart)
      (:Update
        (Variables|Default|SetWeakPointActive _weakpointactive)))))

(event EventBeginPlay
  (|CallParentFunction) ; This node is disabled and will not be called.\nDrag off pins to build functionality.
)

(event EventTick
  (params (DeltaSeconds :type Float%20%28single-precision%29))
  (|CallParentFunction DeltaSeconds) ; This node is disabled and will not be called.\nDrag off pins to build functionality.
)

(fn ConstructionScript
  (|CallParentFunction))
