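; Blueprint for the Pipershelf dual-wield pickaxe cosmetic, adjusting its material look for frontend display.
; On weapon cosmetic init it runs the parent logic, then checks whether it is displayed in a frontend context.
; If so, it sets diamond emissive and caustics intensity parameters on both main and offhand weapon mesh materials.
; The construction script only calls its parent.
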(event EventOnInitWeaponCosmetics
  (bind _weaponmesh (Variables|Visual|GetWeaponMesh))
  (bind _weaponmeshoffhand (Variables|Visual|GetWeaponMeshOffhand))
  (|CallParentFunction)
  (if (Reactive|Frontend|IsCosmeticObjectDisplayedinFrontendContext self)
    (Rendering|Material|SetScalarParameterValueonMaterials _weaponmesh "DiamondEmissiveIntensity" 3.6)
    (Rendering|Material|SetScalarParameterValueonMaterials _weaponmeshoffhand "DiamondEmissiveIntensity" 3.6)
    (Rendering|Material|SetScalarParameterValueonMaterials _weaponmesh "CausticsIntensity" 0.2)
    (Rendering|Material|SetScalarParameterValueonMaterials _weaponmeshoffhand "CausticsIntensity" 0.2)))

(fn ConstructionScript
  (|CallParentFunction))
