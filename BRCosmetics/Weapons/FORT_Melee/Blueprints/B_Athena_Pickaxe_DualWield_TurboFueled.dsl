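; Blueprint for the TurboFueled dual-wield pickaxe cosmetic, adding special anim trail handling for the last swing of its attack.
; It overrides the anim trail notify and disable events, calling the parent first and then reading the instigator pawn's current montage section.
; When that section is "Swing4", it activates the anim trail VFX component on notify, regardless of the notify's active flag, and deactivates it on disable.
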(event EventBeginPlay)

(event EventTick
  (params (DeltaSeconds :type Float%20%28single-precision%29)))

(event EventAnimTrailsNotify
  (params (bActive :type Boolean))
  ;[0.001785 0.2569 0.06994] Fist Pickaxe Last Swing Anim Trail
  (|CallParentFunction bActive)
  (bind _asfort_pawn (Utilities|Casting|CastToFortPawn (Game|GetInstigator)))
  (if (== (Animation|Montage|MontageGetCurrentSection (Components|SkeletalMesh|GetAnimInstance (Variables|/Script/Engine.Character|GetMesh _asfort_pawn)) (Animation|Montage|GetCurrentActiveMontage (Components|SkeletalMesh|GetAnimInstance (Variables|/Script/Engine.Character|GetMesh _asfort_pawn)))) "Swing4")
    (Components|Activation|Activate (CallFunction|GetAnimTrailVFXComponent) true)))

(event EventAnimTrailsDisable
  ;[0.001785 0.2569 0.06994] Fist Pickaxe Last Swing Anim Trail
  (|CallParentFunction)
  (bind _asfort_pawn (Utilities|Casting|CastToFortPawn (Game|GetInstigator)))
  (if (== (Animation|Montage|MontageGetCurrentSection (Components|SkeletalMesh|GetAnimInstance (Variables|/Script/Engine.Character|GetMesh _asfort_pawn)) (Animation|Montage|GetCurrentActiveMontage (Components|SkeletalMesh|GetAnimInstance (Variables|/Script/Engine.Character|GetMesh _asfort_pawn)))) "Swing4")
    (Components|Activation|Deactivate (CallFunction|GetAnimTrailVFXComponent))))

(fn ConstructionScript
  (|CallParentFunction))
