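; Melee pickaxe Blueprint for the FeatherMudGlance cosmetic, adding a reactive "weak spot" effect to impacts on certain surfaces.
; On impact it stores the spawned particle component and triggers the effect directly in frontend or preview, or through server-then-client events when the surface is listed.
; The effect pulses bWeakSpotHit on impact and idle particles for 1.5 seconds, timestamps the weapon mesh materials for shader reactivity and plays an attached sound.
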(event FortWeapon|EventOnPlayImpactFX
  (params (HitResult :type Hit%20Result%20Structure%20%28by%20ref%29) (ImpactPhysicalSurface :type EPhysicalSurface%20Enum :default "SurfaceType_Default") (SpawnedPSC :type FXSystem%20Component%20Object%20Reference))
  (|CallParentFunction HitResult ImpactPhysicalSurface SpawnedPSC)
  (Variables|Default|SetSpawned%20PSC_0 SpawnedPSC)
  (if (or (Reactive|Frontend|IsCosmeticObjectDisplayedinFrontendContext 0) (Reactive|Frontend|IsCosmeticObjectDisplayedasPreview 0))
    (Effects|Components|ParticleSystem|SetBoolParameter (Variables|Default|GetSpawned%20PSC_0) "bWeakSpotHit" true)
    (CallFunction|WeakSpotFXPlay)
    (elif (Utilities|Array|ContainsItem (Variables|Default|GetImpactPhysicalSurface) ImpactPhysicalSurface)
      (Effects|Components|ParticleSystem|SetBoolParameter (Variables|Default|GetSpawned%20PSC_0) "bWeakSpotHit" true)
      (CallFunction|FXServer))))

(event Custom|FX_Server
  (CallFunction|FXClient))

(event Custom|FX_Client
  (CallFunction|WeakSpotFXPlay))

(event Custom|WeakSpotFXPlay
  (Utilities|FlowControl|Sequence
    (:then_0
      (Effects|Components|ParticleSystem|SetBoolParameter (Variables|Default|GetSpawned%20PSC_0) "bWeakSpotHit" true)
      (Effects|Components|ParticleSystem|SetBoolParameter (CallFunction|GetIdleVFXComponent) "bWeakSpotHit" true)
      (Utilities|FlowControl|Delay 1.5)
      (Effects|Components|ParticleSystem|SetBoolParameter (Variables|Default|GetSpawned%20PSC_0) "bWeakSpotHit" false)
      (Effects|Components|ParticleSystem|SetBoolParameter (CallFunction|GetIdleVFXComponent) "bWeakSpotHit" false))
    (:then_1
      ;[1 0.2552 0.2552] Material Reactivity
      (Rendering|Material|SetScalarParameterValueonMaterials (Variables|Visual|GetWeaponMesh) "BP_GetTime" (Utilities|Time|GetGameTimeinSeconds))
      (Audio|SpawnSoundAttached "/Game/Athena/Sounds/Weapons/PickAxes/FeatherMudGlance/MSS_Pickaxe_Reactive_FeatherMudGlance_Impact.MSS_Pickaxe_Reactive_FeatherMudGlance_Impact" (Variables|Visual|GetWeaponMesh)))))

(fn ConstructionScript
  (|CallParentFunction))
