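; Pickaxe weapon Blueprint for the GallonBag cosmetic that drives time-based reactive visuals from impact events.
; On each impact it caches the effect, checks the hit surface against the weak-point surface list, and also triggers a material update when displayed in the frontend.
; Reactions write current game time into a ReactivityTime parameter on the weapon mesh and on idle, swing and impact VFX, letting materials and particles time the response.
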(event EventTick
  (params (DeltaSeconds :type Float%20%28single-precision%29)))

(event EventBeginPlay)

(event FortWeapon|EventOnPlayImpactFX
  (params (HitResult :type Hit%20Result%20Structure%20%28by%20ref%29) (ImpactPhysicalSurface :type EPhysicalSurface%20Enum :default "SurfaceType_Default") (SpawnedPSC :type FXSystem%20Component%20Object%20Reference))
  (Variables|Default|SetImpact%20VFX SpawnedPSC)
  (|CallParentFunction HitResult ImpactPhysicalSurface SpawnedPSC)
  (Utilities|FlowControl|Sequence
    (:then_0
      (if (Reactive|Frontend|IsCosmeticObjectDisplayedinFrontendContext self)
        (CallFunction|MaterialUpdate)))
    (:then_1
      (bind _output_get (Variables|Default|SetCurrent%20Impact%20Physical%20Surface%200 ImpactPhysicalSurface))
      (Variables|Default|SetIs%20weak%20point%20hit (Utilities|Array|ContainsItem (Variables|Default|GetWeakPoint%20Physical%20Surface%20Array_0) _output_get))
      (if (Variables|Default|GetIs%20weak%20point%20hit)
        (Effects|Components|ParticleSystem|SetFloatParameter (Variables|Default|GetImpact%20VFX) "ReactivityTime" (Utilities|Time|GetGameTimeinSeconds))
        (CallFunction|OnWeakPointHit)))))

(event Custom|OnWeakPointHit
  (CallFunction|MaterialUpdate))

(event Custom|MaterialUpdate
  (bind _returnvalue (Utilities|Time|GetGameTimeinSeconds))
  (Rendering|Material|SetScalarParameterValueonMaterials (Variables|Visual|GetWeaponMesh) "ReactivityTime" _returnvalue)
  (Effects|Components|ParticleSystem|SetFloatParameter (CallFunction|GetIdleVFXComponent) "ReactivityTime" _returnvalue)
  (Effects|Components|ParticleSystem|SetFloatParameter (CallFunction|GetSwingVFXComponent) "ReactivityTime" _returnvalue)
  (Effects|Components|ParticleSystem|SetFloatParameter (Variables|Default|GetImpact%20VFX) "ReactivityTime" _returnvalue))

(fn ConstructionScript
  (|CallParentFunction))
