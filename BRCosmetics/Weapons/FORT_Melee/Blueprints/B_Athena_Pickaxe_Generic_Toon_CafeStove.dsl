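; Toon Cafe Stove pickaxe Blueprint that makes the weapon glow reactively on impact.
; BeginPlay applies the toon frontend LUT replacement.
; Impact FX calls the reactivity event directly in frontend display; otherwise it switches on surface type and relays a weak-point hit from server to client.
; Both paths restart one timeline driving mesh emissive intensity and glow opacity on idle and swing-trail VFX.
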(event EventBeginPlay
  ;[0 0.8066 1] Toon LUT for FE
  (|CallParentFunction)
  (Utilities|ToonFrontendLUTReplacePickaxe))

(event FortWeapon|EventOnPlayImpactFX
  (params (HitResult :type Hit%20Result%20Structure%20%28by%20ref%29) (ImpactPhysicalSurface :type EPhysicalSurface%20Enum :default "SurfaceType_Default") (SpawnedPSC :type FXSystem%20Component%20Object%20Reference))
  ;[1 0.0102 0.03009] VFX events
  (|CallParentFunction HitResult ImpactPhysicalSurface SpawnedPSC)
  (if (Reactive|Frontend|IsCosmeticObjectDisplayedinFrontendContext 0)
    (CallFunction|FEReactivety)
    (else
      (switch Utilities|FlowControl|Switch|SwitchonEPhysicalSurface ImpactPhysicalSurface
        (:SurfaceType12
          _)
        (:SurfaceType14
          _)
        (:SurfaceType15
          _)
        (:SurfaceType16
          _))
      (CallFunction|OnWeakPointHitServer))))

(event Custom|OnWeakPointHit_Server
  ;[0.9711 1 0] Send Signal from Server to Client
  (CallFunction|OnWeakPointHitClient))

(event Custom|FE_Reactivety
  ;[0 1 0.1273] Play WeakPointHit on Clients and in FE
  (bind (_direction _glow_intensity) (Timeline|Timeline PlayFromStart)
    (:Update
      (Rendering|Material|SetScalarParameterValueonMaterials (Variables|Visual|GetWeaponMesh) "Emissive_Intensity" (* _glow_intensity 5.0))
      (Effects|Components|ParticleSystem|SetFloatParameter (CallFunction|GetIdleVFXComponent) "GlowOpacity" _glow_intensity)
      (Effects|Components|ParticleSystem|SetFloatParameter (CallFunction|GetAnimTrailVFXComponent) "GlowOpacity" _glow_intensity))))

(event Custom|OnWeakPointHit_Client
  ;[0 1 0.1273] Play WeakPointHit on Clients and in FE
  (Timeline|Timeline PlayFromStart))

(fn ConstructionScript
  (|CallParentFunction))
