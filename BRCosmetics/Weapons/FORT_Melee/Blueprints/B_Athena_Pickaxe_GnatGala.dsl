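; GnatGala pickaxe weapon Blueprint that adjusts where its impact effect appears for each cosmetic mesh variant.
; On impact FX it calls the parent, then, gated by a cosmetic-preview check, reads the pawn's active Mesh variant stage.
; The stage selects one of three hand-set offsets, rotated by the instigator's rotation, as the effect's relative location.
; Separate offsets exist because variants use dual-wield or single-wield animations.
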(event EventBeginPlay)

(event EventTick
  (params (DeltaSeconds :type Float%20%28single-precision%29)))

(event FortWeapon|EventOnPlayImpactFX
  (params (HitResult :type Hit%20Result%20Structure%20%28by%20ref%29) (ImpactPhysicalSurface :type EPhysicalSurface%20Enum :default "SurfaceType_Default") (SpawnedPSC :type FXSystem%20Component%20Object%20Reference))
  (|CallParentFunction HitResult ImpactPhysicalSurface SpawnedPSC)
  (if (Reactive|Frontend|IsCosmeticObjectDisplayedasPreview 0)
    (bind _asfort_player_pawn (Utilities|Casting|CastToFortPlayerPawn (Game|GetInstigator)))
    ; Added custom Impact Locations cuz the Variants are using 2 different Anims. Dual Wield & Single Wield
    (bind _index (CosmeticLoadout|GetVariantChannelActiveIndexforSlotActor "(TagName=\"LoadoutSlot.Pickaxe\")" _asfort_player_pawn "(TagName=\"Cosmetics.Variant.Channel.Mesh\")" (Utilities|Array|MakeArray "(TagName=\"Cosmetics.Variant.Property.Stage1\")" "(TagName=\"Cosmetics.Variant.Property.Stage2\")" "(TagName=\"Cosmetics.Variant.Property.Stage3\")")))
    (Transformation|SetRelativeLocation (Variables|Default|GetImpact%20FX) (Math|Vector|RotateVector (Utilities|Select :Index _index :Option%200 "0.000000,105.000000,138.000000" :Option%201 "0.000000,105.000000,135.000000" :Option%202 "0.000000,117.000000,120.000000") (Transformation|GetActorRotation (Game|GetInstigator))) false false)))

(fn ConstructionScript
  (|CallParentFunction))
