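; Fortnite pickaxe Blueprint for the GnocchiTea cosmetic, driving its idle, swing trail and impact effects.
; Idle effects follow weapon visibility, with construction disabling their auto-activation.
; Swings pass an incremented count to idle VFX for blade blur and the montage section index to trails; in the frontend, an anim Blueprint event replays the blur.
; Impacts pass the hit actor to the effect and select body or general electric discharge emitters by whether it is a pawn.
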(event EventBeginPlay
  (|CallParentFunction)
  (bind _asgnocchi_tea_axe_anim_bp (Utilities|Casting|CastToGnocchiTea_Axe_AnimBP (Components|SkeletalMesh|GetAnimInstance (Variables|Visual|GetWeaponMesh))))
  (if (Reactive|Frontend|IsCosmeticObjectDisplayedinFrontendContext 0)
    (Default|BindEventtoOnSepcialActionFired _asgnocchi_tea_axe_anim_bp (EventDispatchers|CreateEvent :!event Custom|OnFrontendSpecialAction))))

(event EventOnWeaponVisibilityChanged
  (params (bVisible :type Boolean) (bSetForLocalControllerOnly :type Boolean))
  (if bVisible
    (Components|Activation|Activate (CallFunction|GetIdleVFXComponent) false)
    (else
      (Components|Activation|Deactivate (CallFunction|GetIdlePSC)))))

(event FortWeapon|EventOnPlayImpactFX
  (params (HitResult :type Hit%20Result%20Structure%20%28by%20ref%29) (ImpactPhysicalSurface :type EPhysicalSurface%20Enum :default "SurfaceType_Default") (SpawnedPSC :type FXSystem%20Component%20Object%20Reference))
  (bind (_bblockinghit _binitialoverlap _time _distance _location _impactpoint _normal _impactnormal _physmat _hitactor) (Collision|BreakHitResult HitResult))
  (bind (_asfort_pawn _bsuccess) (Utilities|Casting|CastToFortPawn _hitactor))
  (|CallParentFunction HitResult ImpactPhysicalSurface SpawnedPSC)
  (Effects|Components|ParticleSystem|SetActorParameter SpawnedPSC "BoneSocketActor" _hitactor)
  (if _bsuccess
    (Effects|Components|ParticleSystem|SetEmitterEnable SpawnedPSC "Electric Discharge" false)
    (Effects|Components|ParticleSystem|SetEmitterEnable SpawnedPSC "Electric Discharge_Body" true)
    (else
      (Effects|Components|ParticleSystem|SetEmitterEnable SpawnedPSC "Electric Discharge" true)
      (Effects|Components|ParticleSystem|SetEmitterEnable SpawnedPSC "Electric Discharge_Body" false))))

(event Custom|ShowBlurFX
  ; FX Idle Blades blur activation
  (bind _result (Math|Integer|IncrementInt (Variables|Default|GetSwingCount))
    (:%20%20
      (Effects|Components|ParticleSystem|SetIntParameter (CallFunction|GetIdleVFXComponent) "SwingCount" _result))))

(event EventSwingLeft
  (bind _returnvalue (CallFunction|GetAnimTrailVFXComponent))
  (|CallParentFunction)
  (CallFunction|ShowBlurFX)
  ; SwingSectionIndex Trails
  (bind _section_index (Utilities|SetupSwingMontageSectionIndex (Variables|Actor|GetInstigator))
    (:Out
      (Effects|Components|ParticleSystem|SetIntParameter _returnvalue "SwingSectionIndex" _section_index))
    (:Cast%20Failed%20Out
      (Effects|Components|ParticleSystem|SetIntParameter _returnvalue "SwingSectionIndex" -1))))

(event EventSwingRight
  (bind _returnvalue (CallFunction|GetAnimTrailVFXComponent))
  (|CallParentFunction)
  (CallFunction|ShowBlurFX)
  ; SwingSectionIndex Trails
  (bind _section_index (Utilities|SetupSwingMontageSectionIndex (Variables|Actor|GetInstigator))
    (:Out
      (Effects|Components|ParticleSystem|SetIntParameter _returnvalue "SwingSectionIndex" _section_index))
    (:Cast%20Failed%20Out
      (Effects|Components|ParticleSystem|SetIntParameter _returnvalue "SwingSectionIndex" -1))))

(event Custom|OnFrontendSpecialAction
  (params (ActionName :type Name))
  (switch name ActionName
    (:SpecialPreviewAction_SwingTool
      (CallFunction|ShowBlurFX))))

(fn ConstructionScript
  (|CallParentFunction)
  (Components|Activation|SetAutoActivate (CallFunction|GetIdlePSC) false))
