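; Cosmetic weapon Blueprint for the GroovyReaderPoke Lance pickaxe, giving its first thrust attack a distinct visual effect.
; Left and right swing events run identical logic after the parent call, checking whether the instigator's active montage section is "Swing1".
; Where cosmetic events can execute, Swing1 spawns a thrust Niagara system attached to the weapon mesh and hides the animation trail.
; Other sections show the trail instead.
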(event EventBeginPlay)

(event EventTick
  (params (DeltaSeconds :type Float%20%28single-precision%29)))

(event EventSwingRight
  (bind _returnvalue (Game|GetInstigator))
  (bind _asfort_pawn (Utilities|Casting|CastToFortPawn _returnvalue))
  (bind _mesh (Variables|/Script/Engine.Character|GetMesh _asfort_pawn))
  (bind _returnvalue_2 (Components|SkeletalMesh|GetAnimInstance _mesh))
  (bind _returnvalue_3 (Animation|Montage|GetCurrentActiveMontage _returnvalue_2))
  (bind _returnvalue_4 (Animation|Montage|MontageGetCurrentSection _returnvalue_2 _returnvalue_3))
  (bind _returnvalue_5 (== _returnvalue_4 "Swing1"))
  (|CallParentFunction)
  (Networking|CanExecuteCosmeticEvents
    (:True
      (Utilities|FlowControl|Sequence
        (:then_0
          (if _returnvalue_5
            (Niagara|SpawnSystemAttached (Variables|Default|GetFX_Thrust_First_Attack) (Variables|Visual|GetWeaponMesh))))
        (:then_1
          (Development|SetHiddeninGame (CallFunction|GetAnimTrailVFXComponent) _returnvalue_5 false))))))

(event EventSwingLeft
  (bind _returnvalue (Game|GetInstigator))
  (bind _asfort_pawn (Utilities|Casting|CastToFortPawn _returnvalue))
  (bind _mesh (Variables|/Script/Engine.Character|GetMesh _asfort_pawn))
  (bind _returnvalue_2 (Components|SkeletalMesh|GetAnimInstance _mesh))
  (bind _returnvalue_3 (Animation|Montage|GetCurrentActiveMontage _returnvalue_2))
  (bind _returnvalue_4 (Animation|Montage|MontageGetCurrentSection _returnvalue_2 _returnvalue_3))
  (bind _returnvalue_5 (== _returnvalue_4 "Swing1"))
  (|CallParentFunction)
  (Networking|CanExecuteCosmeticEvents
    (:True
      (Utilities|FlowControl|Sequence
        (:then_0
          (if _returnvalue_5
            (Niagara|SpawnSystemAttached (Variables|Default|GetFX_Thrust_First_Attack) (Variables|Visual|GetWeaponMesh))))
        (:then_1
          (Development|SetHiddeninGame (CallFunction|GetAnimTrailVFXComponent) _returnvalue_5 false))))))

(fn ConstructionScript
  (|CallParentFunction))
