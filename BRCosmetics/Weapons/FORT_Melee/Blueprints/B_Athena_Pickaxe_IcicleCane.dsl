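; Fortnite pickaxe Blueprint for the Icicle Cane cosmetic, adding variant-aware swing effects.
; On equip, it reads the owner's pickaxe Particle variant channel and stores the active index as a float.
; Left and right swings set a direction flag and, during montage section "Swing2" where cosmetics run, attach a jab effect fed that "Variant" value.
; A helper passes a swing index to the audio parameter component, with no caller here.
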(event FortWeapon|EventOnEquippedAndAttached
  (|CallParentFunction)
  (bind _asfort_player_pawn (Utilities|Casting|CastToFortPlayerPawn (Variables|Actor|GetInstigator)))
  (bind _index (CosmeticLoadout|GetVariantChannelActiveIndexforSlotActor "(TagName=\"LoadoutSlot.Pickaxe\")" _asfort_player_pawn "(TagName=\"Cosmetics.Variant.Channel.Particle\")" (Utilities|Array|MakeArray "(TagName=\"Cosmetics.Variant.Property.Particle1\")" "(TagName=\"Cosmetics.Variant.Property.Particle2\")")))
  (Variables|Default|SetVariant (Math|Conversions|ToFloat(Integer) _index)))

(event EventSwingLeft
  (|CallParentFunction)
  (Variables|Default|SetRight%20? false)
  (Networking|CanExecuteCosmeticEvents
    (:True
      (if (== (Animation|Montage|MontageGetCurrentSection (Components|SkeletalMesh|GetAnimInstance (Variables|/Script/Engine.Character|GetMesh (Utilities|Casting|CastToFortPawn (Game|GetInstigator)))) (Animation|Montage|GetCurrentActiveMontage (Components|SkeletalMesh|GetAnimInstance (Variables|/Script/Engine.Character|GetMesh (Utilities|Casting|CastToFortPawn (Game|GetInstigator)))))) "Swing2")
        (bind _returnvalue (Effects|SpawnFXSystemAttached (Variables|Default|GetFXSystemJab) (Variables|Visual|GetWeaponMesh) "Melee_TwoHanded_Handle"))
        (Effects|Components|ParticleSystem|SetFloatParameter _returnvalue "Variant" (Variables|Default|GetVariant))))))

(event EventSwingRight
  (|CallParentFunction)
  (Variables|Default|SetRight%20? true)
  (Networking|CanExecuteCosmeticEvents
    (:True
      (if (== (Animation|Montage|MontageGetCurrentSection (Components|SkeletalMesh|GetAnimInstance (Variables|/Script/Engine.Character|GetMesh (Utilities|Casting|CastToFortPawn (Game|GetInstigator)))) (Animation|Montage|GetCurrentActiveMontage (Components|SkeletalMesh|GetAnimInstance (Variables|/Script/Engine.Character|GetMesh (Utilities|Casting|CastToFortPawn (Game|GetInstigator)))))) "Swing2")
        (bind _returnvalue (Effects|SpawnFXSystemAttached (Variables|Default|GetFXSystemJab) (Variables|Visual|GetWeaponMesh) "Melee_TwoHanded_Handle"))
        (Effects|Components|ParticleSystem|SetFloatParameter _returnvalue "Variant" (Variables|Default|GetVariant))))))

(fn SetSwingAudioParam
  (params (InInt :type Integer))
  (bind _returnvalue (AudioParameter|GetAudioParameterComponent))
  (Utilities|IsValid _returnvalue
    (:Is%20Valid
      (Audio|Parameter|SetIntegerParameter _returnvalue "SwingIdx" InInt :!class /Script/AudioExtensions.AudioParameterControllerInterface))))

(fn ConstructionScript
  (|CallParentFunction))
