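; Parent Blueprint for Inkville Athena pickaxe cosmetics that applies a frontend-specific color LUT on weapon cosmetic init.
; In frontend display contexts only, it loops over weapon mesh materials, storing existing dynamic instances or creating and assigning new ones.
; Newly created instances whose PBR_Shading is not 1 receive an inverse LUT volume texture and related scalar parameters.
; An embedded note marks this as a stopgap until an automatic LUT tool exists.
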(event EventOnInitWeaponCosmetics
  (bind _currentmid (Variables|Default|GetCurrentMID))
  (|CallParentFunction)
  ;[1 1 1] Front End LUT, Remove after creating auto LUT tool
  (if (Reactive|Frontend|IsCosmeticObjectDisplayedinFrontendContext self)
    (for (_array_element _array_index) (Rendering|Material|GetMaterials (Variables|Visual|GetWeaponMesh))
      (bind _asmaterial_instance_dynamic (Utilities|Casting|CastToMaterialInstanceDynamic _array_element)
        (:then
          (Variables|Default|SetCurrentMID _asmaterial_instance_dynamic))
        (:CastFailed
          (bind _returnvalue (Rendering|Material|CreateDynamicMaterialInstance _array_element "None" "None" :!class /Script/Engine.KismetMaterialLibrary))
          (bind _output_get (Variables|Default|SetCurrentMID _returnvalue))
          (Rendering|Material|SetMaterial (Variables|Visual|GetWeaponMesh) _array_index _output_get)
          (bind _returnvalue_2 (Rendering|Material|GetScalarParameterValue _currentmid "PBR_Shading" :!class /Script/Engine.MaterialInstanceDynamic))
          (if (not (== _returnvalue_2 1.0))
            (Rendering|Material|SetTextureParameterValue _currentmid "LUT Volume Texture" "/Game/Packages/Fortress_SharedMaterials/Anime_Master_Material/VT_InverseLUT_FE.VT_InverseLUT_FE")
            (Rendering|Material|SetScalarParameterValue _currentmid "LUT Multiplier" (Variables|Default|GetLUT%20Multiplier) :!class /Script/Engine.MaterialInstanceDynamic)
            (Rendering|Material|SetScalarParameterValue _currentmid "Mobile LUT Brightness" 1.0 :!class /Script/Engine.MaterialInstanceDynamic)
            (Rendering|Material|SetScalarParameterValue _currentmid "use 33 Resolution" 0.0 :!class /Script/Engine.MaterialInstanceDynamic)))))))

(fn ConstructionScript
  (|CallParentFunction))
