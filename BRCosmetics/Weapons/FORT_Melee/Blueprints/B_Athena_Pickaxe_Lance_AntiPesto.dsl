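; Blueprint for the Lance AntiPesto melee pickaxe cosmetic in Battle Royale, adding a jab effect to its swings.
; Left and right swing events call the parent, then look up the instigator's swing montage section index.
; When that index is 0, the FX_Jab Niagara system spawns attached to the weapon mesh at the two-handed handle socket.
; BeginPlay and Tick are empty, and construction only calls the parent.
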(event EventBeginPlay)

(event EventTick
  (params (DeltaSeconds :type Float%20%28single-precision%29)))

(event EventSwingLeft
  (|CallParentFunction)
  (bind _section_index (Utilities|SetupSwingMontageSectionIndex (Variables|Actor|GetInstigator))
    (:Out
      (if (== _section_index 0)
        (Niagara|SpawnSystemAttached (Variables|Default|GetFX_Jab) (Variables|Visual|GetWeaponMesh) "Melee_TwoHanded_Handle" "0, 0, 0" "0, 0, 0" "KeepRelativeOffset" true)))))

(event EventSwingRight
  (|CallParentFunction)
  (bind _section_index (Utilities|SetupSwingMontageSectionIndex (Variables|Actor|GetInstigator))
    (:Out
      (if (== _section_index 0)
        (Niagara|SpawnSystemAttached (Variables|Default|GetFX_Jab) (Variables|Visual|GetWeaponMesh) "Melee_TwoHanded_Handle" "0, 0, 0" "0, 0, 0" "KeepRelativeOffset" true)))))

(fn ConstructionScript
  (|CallParentFunction))
