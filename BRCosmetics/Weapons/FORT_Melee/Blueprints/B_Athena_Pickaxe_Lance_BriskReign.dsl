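; Cosmetic Blueprint for the BriskReign lance pickaxe that chooses the visual effect for each swing.
; Left and right swing events run identical logic, reading the montage section index from the instigator.
; Section 0 deactivates the anim trail and spawns a jab Niagara effect on the weapon handle, passing it the active particle variant as a float.
; Other sections send the index to the trail as SwingSectionIndex, or -1 when the cast fails.
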(event EventSwingLeft
  (bind _returnvalue (CallFunction|GetAnimTrailVFXComponent))
  (|CallParentFunction)
  (bind _section_index (Utilities|SetupSwingMontageSectionIndex (Variables|Actor|GetInstigator))
    (:Out
      (if (== _section_index 0)
        (Components|Activation|Deactivate _returnvalue)
        (bind _returnvalue_2 (Niagara|SpawnSystemAttached (Variables|Default|GetFX_Jab) (Variables|Visual|GetWeaponMesh) "Melee_TwoHanded_Handle" "0, 0, 0" "0, 0, 0" "KeepRelativeOffset" true))
        (bind _index (Class|FortCosmeticFunctionLibrary|%28Deprecated%29GetVariantChannelActiveIndexForSlot "(TagName=\"LoadoutSlot.Pickaxe\")" (CallFunction|GetCosmeticData) "(TagName=\"Cosmetics.Variant.Channel.Particle\")" (Utilities|Array|MakeArray "(TagName=\"Cosmetics.Variant.Property.Particle1\")" "(TagName=\"Cosmetics.Variant.Property.Particle3\")" "(TagName=\"Cosmetics.Variant.Property.Particle2\")")))
        (Niagara|SetNiagaraVariable(Float) _returnvalue_2 "Variant" (Math|Conversions|ToFloat(Integer) _index))
        (else
          (Effects|Components|ParticleSystem|SetIntParameter _returnvalue "SwingSectionIndex" _section_index))))
    (:Cast%20Failed%20Out
      (Effects|Components|ParticleSystem|SetIntParameter _returnvalue "SwingSectionIndex" -1))))

(event EventSwingRight
  (bind _returnvalue (CallFunction|GetAnimTrailVFXComponent))
  (|CallParentFunction)
  (bind _section_index (Utilities|SetupSwingMontageSectionIndex (Variables|Actor|GetInstigator))
    (:Out
      (if (== _section_index 0)
        (Components|Activation|Deactivate _returnvalue)
        (bind _returnvalue_2 (Niagara|SpawnSystemAttached (Variables|Default|GetFX_Jab) (Variables|Visual|GetWeaponMesh) "Melee_TwoHanded_Handle" "0, 0, 0" "0, 0, 0" "KeepRelativeOffset" true))
        (bind _index (Class|FortCosmeticFunctionLibrary|%28Deprecated%29GetVariantChannelActiveIndexForSlot "(TagName=\"LoadoutSlot.Pickaxe\")" (CallFunction|GetCosmeticData) "(TagName=\"Cosmetics.Variant.Channel.Particle\")" (Utilities|Array|MakeArray "(TagName=\"Cosmetics.Variant.Property.Particle1\")" "(TagName=\"Cosmetics.Variant.Property.Particle3\")" "(TagName=\"Cosmetics.Variant.Property.Particle2\")")))
        (Niagara|SetNiagaraVariable(Float) _returnvalue_2 "Variant" (Math|Conversions|ToFloat(Integer) _index))
        (else
          (Effects|Components|ParticleSystem|SetIntParameter _returnvalue "SwingSectionIndex" _section_index))))
    (:Cast%20Failed%20Out
      (Effects|Components|ParticleSystem|SetIntParameter _returnvalue "SwingSectionIndex" -1))))

(fn ConstructionScript
  (|CallParentFunction))
