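; Fortnite Athena lance pickaxe cosmetic (ChiliAbide) that picks swing VFX by montage section.
; Left and right swing events share identical logic, calling the parent, then reading the instigator's swing montage section index.
; Section zero deactivates the anim trail and attaches a Niagara jab effect to the weapon mesh's handle socket.
; Other sections pass the index to the trail's SwingSectionIndex parameter, or -1 when the lookup's cast fails.
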(event EventSwingLeft
  (bind _returnvalue (CallFunction|GetAnimTrailVFXComponent))
  (|CallParentFunction)
  (bind _section_index (Utilities|SetupSwingMontageSectionIndex (Variables|Actor|GetInstigator))
    (:Out
      (if (== _section_index 0)
        (Components|Activation|Deactivate _returnvalue)
        (Niagara|SpawnSystemAttached (Variables|Default|GetFX_Jab) (Variables|Visual|GetWeaponMesh) "Melee_TwoHanded_Handle" "0, 0, 0" "0, 0, 0" "KeepRelativeOffset" true)
        (else
          (Effects|Components|ParticleSystem|SetIntParameter _returnvalue "SwingSectionIndex" _section_index))))
    (:Cast%20Failed%20Out
      (Effects|Components|ParticleSystem|SetIntParameter _returnvalue "SwingSectionIndex" -1))))

(event EventSwingRight
  (bind _returnvalue (CallFunction|GetAnimTrailVFXComponent))
  (|CallParentFunction)
  (bind _section_index (Utilities|SetupSwingMontageSectionIndex (Variables|Actor|GetInstigator))
    (:Out
      (if (== _section_index 0)
        (Components|Activation|Deactivate _returnvalue)
        (Niagara|SpawnSystemAttached (Variables|Default|GetFX_Jab) (Variables|Visual|GetWeaponMesh) "Melee_TwoHanded_Handle" "0, 0, 0" "0, 0, 0" "KeepRelativeOffset" true)
        (else
          (Effects|Components|ParticleSystem|SetIntParameter _returnvalue "SwingSectionIndex" _section_index))))
    (:Cast%20Failed%20Out
      (Effects|Components|ParticleSystem|SetIntParameter _returnvalue "SwingSectionIndex" -1))))

(fn ConstructionScript
  (|CallParentFunction))
