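; Child Blueprint for the GlowFang lance pickaxe cosmetic that suppresses the swing VFX on one section of the swing montage.
; On left and right swing events, it calls the parent logic, then gets the swing montage section index from the instigator.
; When that index is 0, it deactivates the swing VFX component.
; BeginPlay and Tick hold only disabled stubs, and the construction script just calls the parent.
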(event EventBeginPlay
  (|CallParentFunction) ; This node is disabled and will not be called.\nDrag off pins to build functionality.
)

(event EventTick
  (params (DeltaSeconds :type Float%20%28single-precision%29))
  (|CallParentFunction DeltaSeconds) ; This node is disabled and will not be called.\nDrag off pins to build functionality.
)

(event EventSwingLeft
  (|CallParentFunction)
  (bind _section_index (Utilities|SetupSwingMontageSectionIndex (Variables|Actor|GetInstigator))
    (:Out
      (if (== _section_index 0)
        (Components|Activation|Deactivate (CallFunction|GetSwingVFXComponent))))))

(event EventSwingRight
  (|CallParentFunction)
  (bind _section_index (Utilities|SetupSwingMontageSectionIndex (Variables|Actor|GetInstigator))
    (:Out
      (if (== _section_index 0)
        (Components|Activation|Deactivate (CallFunction|GetSwingVFXComponent))))))

(fn ConstructionScript
  (|CallParentFunction))
