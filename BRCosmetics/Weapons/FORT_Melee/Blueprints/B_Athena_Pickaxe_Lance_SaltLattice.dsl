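; Athena cosmetic pickaxe Blueprint for the SaltLattice lance that plays a variant-aware Niagara jab effect on swings.
; On equip, it looks up the active Particle variant channel index from the owning pawn's pickaxe loadout slot and stores it as a float.
; Left and right swings at montage section index zero spawn the jab system on the weapon mesh's two-handed handle socket, passing that float as its Variant parameter.
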(event FortWeapon|EventOnEquippedAndAttached
  (|CallParentFunction)
  (bind _asfort_player_pawn (Utilities|Casting|CastToFortPlayerPawn (Variables|Actor|GetInstigator)))
  (bind _index (CosmeticLoadout|GetVariantChannelActiveIndexforSlotActor "(TagName=\"LoadoutSlot.Pickaxe\")" _asfort_player_pawn "(TagName=\"Cosmetics.Variant.Channel.Particle\")" (Utilities|Array|MakeArray "(TagName=\"Cosmetics.Variant.Property.Particle1\")" "(TagName=\"Cosmetics.Variant.Property.Particle2\")" "(TagName=\"Cosmetics.Variant.Property.Particle3\")" "(TagName=\"Cosmetics.Variant.Property.Particle4\")" "(TagName=\"Cosmetics.Variant.Property.Particle5\")" "(TagName=\"Cosmetics.Variant.Property.Particle6\")")))
  (Variables|Default|SetVariant (Math|Conversions|ToFloat(Integer) _index)))

(event EventSwingLeft
  (|CallParentFunction)
  (bind _section_index (Utilities|SetupSwingMontageSectionIndex (Variables|Actor|GetInstigator))
    (:Out
      (if (== _section_index 0)
        (bind _returnvalue (Niagara|SpawnSystemAttached (Variables|Default|GetFX_Jab) (Variables|Visual|GetWeaponMesh) "Melee_TwoHanded_Handle" "0, 0, 0" "0, 0, 0" "KeepRelativeOffset" true))
        (Effects|Components|ParticleSystem|SetFloatParameter _returnvalue "Variant" (Variables|Default|GetVariant))))))

(event EventSwingRight
  (|CallParentFunction)
  (bind _section_index (Utilities|SetupSwingMontageSectionIndex (Variables|Actor|GetInstigator))
    (:Out
      (if (== _section_index 0)
        (bind _returnvalue (Niagara|SpawnSystemAttached (Variables|Default|GetFX_Jab) (Variables|Visual|GetWeaponMesh) "Melee_TwoHanded_Handle" "0, 0, 0" "0, 0, 0" "KeepRelativeOffset" true))
        (Effects|Components|ParticleSystem|SetFloatParameter _returnvalue "Variant" (Variables|Default|GetVariant))))))

(fn ConstructionScript
  (|CallParentFunction))
