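; Cosmetic pickaxe Blueprint (LemonWrath Ochre) that makes the weapon mesh glow in reaction to hits.
; Each impact, or cosmetic init in a frontend context, raises HitCount (capped at 4) and drives material scalars ToggleGlow and HitCount.
; After a two-second delay it switches the glow off, stamps transition parameters, prints debug text and starts a looping four-second timer that zeroes HitCount.
; The impact and frontend paths duplicate this logic.
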(event EventOnInitWeaponCosmetics
  (if (Reactive|Frontend|IsCosmeticObjectDisplayedinFrontendContext 0)
    (CallFunction|HitReactivity)))

(event Custom|HitReactivity
  (Variables|Default|SetHitCount (Math|Integer|Min(Integer) (+ (Variables|Default|GetHitCount) 1) 4))
  (Utilities|Time|ClearandInvalidateTimerbyHandle (Variables|Default|GetHitCountTimerReset))
  (Rendering|Material|SetScalarParameterValueonMaterials (Variables|Visual|GetWeaponMesh) "ToggleGlow" 1.0)
  (Rendering|Material|SetScalarParameterValueonMaterials (Variables|Visual|GetWeaponMesh) "HitCount" (Math|Conversions|ToFloat(Integer) (Variables|Default|GetHitCount)))
  (Utilities|FlowControl|Delay 2.0)
  (Rendering|Material|SetScalarParameterValueonMaterials (Variables|Visual|GetWeaponMesh) "ToggleGlow" 0.0)
  (Rendering|Material|SetScalarParameterValueonMaterials (Variables|Visual|GetWeaponMesh) "TimestampB" (Utilities|Time|GetGameTimeinSeconds))
  (Rendering|Material|SetScalarParameterValueonMaterials (Variables|Visual|GetWeaponMesh) "TransitionLength_InB" 2.0)
  (Development|PrintString " The Hit level is :" true true "(R=0.000000,G=0.660000,B=1.000000,A=1.000000)" 4.0)
  (Development|PrintString (Utilities|String|ToString(Integer) (Variables|Default|GetHitCount)) true true "(R=0.000000,G=0.660000,B=1.000000,A=1.000000)" 4.0)
  (bind _returnvalue (Utilities|Time|SetTimerbyEvent (delegate Custom|ResetHitCountTimer) 4.0 true true))
  (Variables|Default|SetHitCountTimerReset _returnvalue))

(event FortWeapon|EventOnPlayImpactFX
  (params (HitResult :type Hit%20Result%20Structure%20%28by%20ref%29) (ImpactPhysicalSurface :type EPhysicalSurface%20Enum :default "SurfaceType_Default") (SpawnedPSC :type FXSystem%20Component%20Object%20Reference))
  (Variables|Default|SetImpact%20VFX SpawnedPSC)
  (|CallParentFunction HitResult ImpactPhysicalSurface SpawnedPSC)
  (Variables|Default|SetHitCount (Math|Integer|Min(Integer) (+ (Variables|Default|GetHitCount) 1) 4))
  (Utilities|Time|ClearandInvalidateTimerbyHandle (Variables|Default|GetHitCountTimerReset))
  (Rendering|Material|SetScalarParameterValueonMaterials (Variables|Visual|GetWeaponMesh) "ToggleGlow" 1.0)
  (Rendering|Material|SetScalarParameterValueonMaterials (Variables|Visual|GetWeaponMesh) "HitCount" (Math|Conversions|ToFloat(Integer) (Variables|Default|GetHitCount)))
  (Utilities|FlowControl|Delay 2.0)
  (Rendering|Material|SetScalarParameterValueonMaterials (Variables|Visual|GetWeaponMesh) "ToggleGlow" 0.0)
  (Rendering|Material|SetScalarParameterValueonMaterials (Variables|Visual|GetWeaponMesh) "TimestampB" (Utilities|Time|GetGameTimeinSeconds))
  (Rendering|Material|SetScalarParameterValueonMaterials (Variables|Visual|GetWeaponMesh) "TransitionLength_InB" 2.0)
  (Development|PrintString " The Hit level is :" true true "(R=0.000000,G=0.660000,B=1.000000,A=1.000000)" 4.0)
  (Development|PrintString (Utilities|String|ToString(Integer) (Variables|Default|GetHitCount)) true true "(R=0.000000,G=0.660000,B=1.000000,A=1.000000)" 4.0)
  (bind _returnvalue (Utilities|Time|SetTimerbyEvent (delegate Custom|ResetHitCountTimer) 4.0 true true))
  (Variables|Default|SetHitCountTimerReset _returnvalue))

(event Custom|ResetHitCountTimer
  (Variables|Default|SetHitCount 0))

(fn ConstructionScript
  (|CallParentFunction))
