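; Pickaxe cosmetic Blueprint for MaltBurger that makes the weapon's idle VFX glow when it hits something.
; Each impact calls Activate Glow, which plays a timeline driving the GlowAlpha parameter on the idle particle component.
; A hit while the timeline is playing jumps it to a set position instead of restarting it.
; In frontend or preview displays, the finished timeline retriggers itself after a short delay, looping the glow.
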(event FortWeapon|EventOnPlayImpactFX
  (params (HitResult :type Hit%20Result%20Structure%20%28by%20ref%29) (ImpactPhysicalSurface :type EPhysicalSurface%20Enum :default "SurfaceType_Default") (SpawnedPSC :type FXSystem%20Component%20Object%20Reference))
  (|CallParentFunction HitResult ImpactPhysicalSurface SpawnedPSC)
  (CallFunction|ActivateGlow))

(event Custom|Activate%20Glow
  ; Glow Activation
  (if (Components|Timeline|IsPlaying (Variables|B_Athena_Pickaxe_MaltBurger|GetTL%20Glow))
    (CallFunction|RefreshGlow)
    (else
      (bind (_direction _alpha) (Timeline|TL%20Glow PlayFromStart)
        (:Update
          (Effects|Components|ParticleSystem|SetFloatParameter (CallFunction|GetIdleVFXComponent) "GlowAlpha" _alpha))
        (:Finished
          (if (or (Reactive|Frontend|IsCosmeticObjectDisplayedinFrontendContext 0) (Reactive|Frontend|IsCosmeticObjectDisplayedasPreview 0))
            (Utilities|FlowControl|Delay 3.0)
            (CallFunction|ActivateGlow)))))))

(event Custom|Refresh%20Glow
  ; Glow Activation
  (Components|Timeline|SetPlaybackPosition (Variables|B_Athena_Pickaxe_MaltBurger|GetTL%20Glow) 0.75 false true))

(fn ConstructionScript
  (|CallParentFunction))
