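; Reactive pickaxe Blueprint for the MarkerDeer Tin cosmetic, escalating a hit counter on weakpoint hits, or on any impact in preview and frontend displays.
; Each reactivity event increments the counter, updates impact, swing and idle VFX parameters, and plays an attached reactive sound.
; It also passes game time and hit count to the weapon's dynamic material.
; Every seventh hit stamps a material reset time and zeroes the counter.
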(event FortWeapon|EventOnPlayImpactFX
  (params (HitResult :type Hit%20Result%20Structure%20%28by%20ref%29) (ImpactPhysicalSurface :type EPhysicalSurface%20Enum :default "SurfaceType_Default") (SpawnedPSC :type FXSystem%20Component%20Object%20Reference))
  (Variables|Default|SetImpact%20VFX SpawnedPSC)
  (|CallParentFunction HitResult ImpactPhysicalSurface SpawnedPSC)
  (Utilities|FlowControl|Sequence
    (:then_0
      (if (or (Reactive|Frontend|IsCosmeticObjectDisplayedasPreview 0) (Reactive|Frontend|IsCosmeticObjectDisplayedinFrontendContext 0))
        (CallFunction|ReactivityEvent)))))

(event EventOnPlayWeakpointHitFX
  (CallFunction|ReactivityEvent))

(event Custom|ReactivityEvent
  (bind _weaponmesh (Variables|Visual|GetWeaponMesh))
  (bind _material_instance_dynamic (Class|BTechArtGlobalLibrary|GetDynamicMaterialInstance _weaponmesh))
  (bind _returnvalue (Utilities|Time|GetGameTimeinSeconds))
  (Math|Integer|IncrementInt (Variables|Default|GetReactivityCounter)
    (:%20%20
      ; VFX
      (Effects|Components|ParticleSystem|SetBoolParameter (Variables|Default|GetImpact%20VFX) "Reactive" true)
      (Effects|Components|ParticleSystem|SetFloatParameter (CallFunction|GetSwingVFXComponent) "ReactivityTime" (Utilities|Time|GetGameTimeinSeconds))
      (Effects|Components|ParticleSystem|SetIntParameter (CallFunction|GetIdleVFXComponent) "ReactivityCounter" (Variables|Default|GetReactivityCounter))
      ; SFX
      (Audio|SpawnSoundAttached "/BRCosmetics/Sounds/PickAxes/MarkerDeer_Tin/IP_MSS_Pickaxe_Reactive_MarkerDeer_Tin_Impact.IP_MSS_Pickaxe_Reactive_MarkerDeer_Tin_Impact" (Variables|Visual|GetWeaponMesh))
      ; TA
      (Rendering|Material|SetScalarParameterValue _material_instance_dynamic "MDT_GameTimeFromBP" _returnvalue :!class /Script/Engine.MaterialInstanceDynamic)
      (Rendering|Material|SetScalarParameterValue _material_instance_dynamic "MDT_CurrHitCount" (Math|Conversions|ToFloat(Integer) (Variables|Default|GetReactivityCounter)) :!class /Script/Engine.MaterialInstanceDynamic)
      (if (== (Math|Integer|%25%28Integer%29 (Variables|Default|GetReactivityCounter) 7) 0)
        (Rendering|Material|SetScalarParameterValue _material_instance_dynamic "MDT_GameTimeFromBP_Reset" _returnvalue :!class /Script/Engine.MaterialInstanceDynamic)
        (Variables|Default|SetReactivityCounter 0)
        (else
          (Rendering|Material|SetScalarParameterValue _material_instance_dynamic "MDT_GameTimeFromBP_Reset" 0.0 :!class /Script/Engine.MaterialInstanceDynamic))))))

(fn ConstructionScript
  (|CallParentFunction))
