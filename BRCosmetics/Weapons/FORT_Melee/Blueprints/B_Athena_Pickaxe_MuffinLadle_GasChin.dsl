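; Pickaxe cosmetic Blueprint (MuffinLadle GasChin) that makes swing and impact VFX react to swing section and weak-point hits.
; Fire FX passes the instigator's montage section index to the swing VFX, or -1 on cast failure.
; Impact FX caches the spawned effect and sets its weak-point flag through reactivity events when the surface matches the weak-point list.
; Frontend and preview instead fake weak points on section 3, with a leftover debug print.
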(event EventTick
  (params (DeltaSeconds :type Float%20%28single-precision%29)))

(event EventBeginPlay)

(event FortWeapon|EventOnPlayImpactFX
  (params (HitResult :type Hit%20Result%20Structure%20%28by%20ref%29) (ImpactPhysicalSurface :type EPhysicalSurface%20Enum :default "SurfaceType_Default") (SpawnedPSC :type FXSystem%20Component%20Object%20Reference))
  (Variables|Default|SetImpact%20VFX SpawnedPSC)
  (|CallParentFunction HitResult ImpactPhysicalSurface SpawnedPSC)
  (if (or (Reactive|Frontend|IsCosmeticObjectDisplayedinFrontendContext 0) (Reactive|Frontend|IsCosmeticObjectDisplayedasPreview 0))
    (bind _section_index (Utilities|SetupSwingMontageSectionIndex (Game|GetInstigator))
      (:Out
        (Effects|Components|ParticleSystem|SetBoolParameter (Variables|Default|GetImpact%20VFX) "IsWeakPoint" (== _section_index 3))
        (Development|PrintString (Utilities|String|ToString(Boolean) (== _section_index 3)))))
    (else
      (bind _output_get (Variables|Default|SetCurrent%20Impact%20Physical%20Surface%200 ImpactPhysicalSurface))
      (Variables|Default|SetIs%20weak%20point%20hit (Utilities|Array|ContainsItem (Variables|Default|GetWeakPoint%20Physical%20Surface%20Array_0) _output_get))
      (if (Variables|Default|GetIs%20weak%20point%20hit)
        (CallFunction|TriggerReactivity)
        (CallFunction|ReplicateReactivity)))))

(event Custom|ReplicateReactivity
  (CallFunction|TriggerReactivity))

(event EventOnPlayWeaponFireFX
  (params (bPersistentFire :type Boolean) (bSecondaryFire :type Boolean))
  (bind _returnvalue (CallFunction|GetSwingVFXComponent))
  (bind _section_index (Utilities|SetupSwingMontageSectionIndex (Variables|Actor|GetInstigator))
    (:Out
      (Effects|Components|ParticleSystem|SetIntParameter _returnvalue "SwingSectionIndex" _section_index))
    (:Cast%20Failed%20Out
      (Effects|Components|ParticleSystem|SetIntParameter _returnvalue "SwingSectionIndex" -1))))

(event Custom|TriggerReactivity
  (Effects|Components|ParticleSystem|SetBoolParameter (Variables|Default|GetImpact%20VFX) "isWeakPoint" true))

(fn ConstructionScript
  (|CallParentFunction))
