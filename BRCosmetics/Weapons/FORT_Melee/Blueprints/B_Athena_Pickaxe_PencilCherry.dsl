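; Pencil Cherry pickaxe weapon Blueprint that customizes its color, outline and shading after cosmetic initialization.
; It copies the overlay material's CustomColor to the body material's secondary color and offhand trail Niagara effect, and sets grid hue, saturation and brightness parameters.
; Loadout meta tags, read through a deprecated library call, toggle the offhand mesh's outline overlay and switch the main mesh between PBR and toon shading.
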(event EventHandlePostCosmeticInitialization
  (bind _weaponmesh (Variables|Visual|GetWeaponMesh))
  (bind _material_instance_dynamic (Class|BTechArtGlobalLibrary|GetDynamicMaterialInstance _weaponmesh))
  (bind _returnvalue (CallFunction|GetCosmeticData))
  (bind _weaponmesh_2 (Variables|Visual|GetWeaponMesh))
  (bind _returnvalue_2 (Rendering|Material|GetOverlayMaterial _weaponmesh_2 false 0 :!class /Script/Engine.MeshComponent))
  (bind _returnvalue_3 (Class|FortCosmeticFunctionLibrary|%28Deprecated%29GetLoadoutMetaTagsForItem _returnvalue "/BRCosmetics/Athena/Items/Cosmetics/Pickaxes/Pickaxe_PencilCherry.Pickaxe_PencilCherry"))
  (bind _weaponmeshoffhand (Variables|Visual|GetWeaponMeshOffhand))
  (bind _weaponmesh_3 (Variables|Visual|GetWeaponMesh))
  (|CallParentFunction)
  (Utilities|FlowControl|Sequence
    (:then_0
      ;[1 1 1] Init
      (Utilities|IsValid (Variables|Default|GetOverlay%20Material)
        (:Is%20Not%20Valid
          (Variables|Default|SetOverlay%20Material (Rendering|Material|GetOverlayMaterial (Variables|Visual|GetWeaponMesh) false 0 :!class /Script/Engine.MeshComponent)))))
    (:then_1
      ;[1 1 1] Update secondary color
      (Rendering|Material|SetOverlayMaterial (Variables|Visual|GetWeaponMesh) _returnvalue_2 false 0 :!class /Script/Engine.MeshComponent) ; Activate Overlay Materials
      (bind _asmaterial_instance_dynamic (Utilities|Casting|CastToMaterialInstanceDynamic _returnvalue_2))
      (bind _returnvalue_4 (Rendering|Material|GetVectorParameterValue _asmaterial_instance_dynamic "CustomColor" :!class /Script/Engine.MaterialInstanceDynamic))
      (Rendering|Material|SetVectorParameterValue _material_instance_dynamic "ColorLit_G_Mask" _returnvalue_4 :!class /Script/Engine.MaterialInstanceDynamic)
      (Effects|Components|ParticleSystem|SetColorParameter (CallFunction|GetAnimTrailsOffhandPSC) "VaCustomColor" _returnvalue_4) ; Handles the Niagara FX Color Change. The default fallback is handled\n inside the niagara system at system level.
      ;[0.07812 0 0] Adjust grid color for dark and red primary colors:
      (bind _returnvalue_5 (Rendering|Material|GetVectorParameterValue _material_instance_dynamic "ColorLit_R_Mask" :!class /Script/Engine.MaterialInstanceDynamic))
      (if (and (and (<= (Math|Color|BreakColor _returnvalue_5) 0.11) (<= (.G (Math|Color|BreakColor _returnvalue_5)) 0.11)) (<= (.B (Math|Color|BreakColor _returnvalue_5)) 0.11))
        _
        (elif (and (>= (Math|Color|BreakColor _returnvalue_5) 0.1) (<= (.B (Math|Color|BreakColor _returnvalue_5)) 0.04))
          _))
      (Rendering|Material|SetScalarParameterValue _material_instance_dynamic "PencilPickaxe_Hueshift" -0.04 :!class /Script/Engine.MaterialInstanceDynamic)
      (Rendering|Material|SetScalarParameterValue _material_instance_dynamic "PencilPickaxe_Saturation" -0.5 :!class /Script/Engine.MaterialInstanceDynamic)
      (Rendering|Material|SetScalarParameterValue _material_instance_dynamic "PencilPickaxe_Brightness" 2.5 :!class /Script/Engine.MaterialInstanceDynamic))
    (:then_2
      ;[1 1 1] Outline toggle
      (if (GameplayTags|HasTag _returnvalue_3 "(TagName=\"Cosmetics.MetaTags.Pencil.OutlinesOff\")" true)
        (Rendering|Material|SetOverlayMaterial _weaponmeshoffhand 0 false 0 :!class /Script/Engine.MeshComponent) ; Disable Overlay Materials
        (else
          (Rendering|Material|SetOverlayMaterial _weaponmeshoffhand _returnvalue_2 false 0 :!class /Script/Engine.MeshComponent) ; Activate Overlay Materials
        )))
    (:then_3
      (if (GameplayTags|HasTag _returnvalue_3 "(TagName=\"Cosmetics.MetaTags.Pencil.PBR\")" true)
        ;[1 1 1] Shading toggle
        (Rendering|Material|SetScalarParameterValueonMaterials _weaponmesh_3 "PBR_Shading" 1.0) ; PBR Shading On
        (elif (GameplayTags|HasTag _returnvalue_3 "(TagName=\"Cosmetics.MetaTags.Pencil.Toon\")" true)
          (Rendering|Material|SetScalarParameterValueonMaterials _weaponmesh_3 "PBR_Shading" 0.0) ; Toon Shading On
        )))))

(event FortWeapon|EventOnEquippedAndAttached
  (|CallParentFunction))

(fn ConstructionScript
  (|CallParentFunction))
