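; Pickaxe weapon Blueprint for the RadioPaca cosmetic that fakes audio-reactive visuals when shown in frontend or preview contexts.
; After parent cosmetic init, it plays a timeline whose value feeds the EnableFrontendFakeAudioLoop parameter on weapon mesh materials and idle VFX.
; The timeline's finish path uses on and off delay lengths, with a BeginLoop flag skipping the off delay on the first pass.
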(event Custom|OnFakeLoopTimelineUpdate
  (params (TransitionValue :type Float%20%28single-precision%29))
  (Rendering|Material|SetScalarParameterValueonMaterials (Variables|Visual|GetWeaponMesh) "EnableFrontendFakeAudioLoop" TransitionValue) ; Enable Fake Audio in Frontend for Materials
  (Effects|Components|ParticleSystem|SetFloatParameter (CallFunction|GetIdleVFXComponent) "EnableFrontendFakeAudioLoop" TransitionValue))

(event EventBeginPlay)

(event EventTick
  (params (DeltaSeconds :type Float%20%28single-precision%29)))

(event EventOnInitWeaponCosmetics
  (|CallParentFunction)
  (if (or (Reactive|Frontend|IsCosmeticObjectDisplayedinFrontendContext 0) (Reactive|Frontend|IsCosmeticObjectDisplayedasPreview 0))
    (Variables|Default|SetBeginLoop true)
    (bind (_direction _in) (Timeline|Timeline PlayFromStart)
      (:Update
        _)
      (:Finished
        (if (not (Variables|Default|GetBeginLoop))
          (Utilities|FlowControl|Delay (Variables|Default|GetFakeAmplitudeOffLength)))
        (Variables|Default|SetBeginLoop false)
        (Utilities|FlowControl|Delay (Variables|Default|GetFakeAmplitudeOnLength))))
    (CallFunction|OnFakeLoopTimelineUpdate _in)))

(fn ConstructionScript
  (|CallParentFunction))
