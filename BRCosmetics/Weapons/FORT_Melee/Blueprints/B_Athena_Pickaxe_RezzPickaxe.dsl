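; Blueprint for the Rezz pickaxe cosmetic that customizes melee impact effects through the particle parameter bWeakSpotHit.
; On impact it stores the spawned particle component and enables that parameter immediately when shown in the frontend or as a preview.
; Otherwise, if the impact surface is in a configured list, it enables the parameter locally, then calls FX_Server, which chains through FX_Client to WeakSpotFXPlay to set it again.
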(event FortWeapon|EventOnPlayImpactFX
  (params (HitResult :type Hit%20Result%20Structure%20%28by%20ref%29) (ImpactPhysicalSurface :type EPhysicalSurface%20Enum :default "SurfaceType_Default") (SpawnedPSC :type FXSystem%20Component%20Object%20Reference))
  (|CallParentFunction HitResult ImpactPhysicalSurface SpawnedPSC)
  (Variables|Default|SetSpawned%20PSC_0 SpawnedPSC)
  (if (or (Reactive|Frontend|IsCosmeticObjectDisplayedinFrontendContext 0) (Reactive|Frontend|IsCosmeticObjectDisplayedasPreview 0))
    (Effects|Components|ParticleSystem|SetBoolParameter (Variables|Default|GetSpawned%20PSC_0) "bWeakSpotHit" true)
    (elif (Utilities|Array|ContainsItem (Variables|Default|GetImpactPhysicalSurface) ImpactPhysicalSurface)
      (Effects|Components|ParticleSystem|SetBoolParameter (Variables|Default|GetSpawned%20PSC_0) "bWeakSpotHit" true)
      (CallFunction|FXServer))))

(event Custom|FX_Server
  (CallFunction|FXClient))

(event Custom|FX_Client
  (CallFunction|WeakSpotFXPlay))

(event Custom|WeakSpotFXPlay
  (Effects|Components|ParticleSystem|SetBoolParameter (Variables|Default|GetSpawned%20PSC_0) "bWeakSpotHit" true))

(fn ConstructionScript
  (|CallParentFunction))
