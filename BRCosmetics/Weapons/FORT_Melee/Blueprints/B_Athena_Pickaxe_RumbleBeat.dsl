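; Cosmetic Blueprint for the RumbleBeat pickaxe that makes the weapon glow and pulse its idle VFX when it hits something.
; Impacts raise the mesh's ToggleGlow for a second, then timestamp a material transition, while activating the VFX with HasHitThing.
; One timeline drives ReactiveCurve and clears HasHitThing, while a second deactivates the VFX.
; HitReactivity duplicates this reaction for frontend display, and equipping switches the VFX off after a short delay.
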(event EventOnInitWeaponCosmetics
  (if (Reactive|Frontend|IsCosmeticObjectDisplayedinFrontendContext 0)
    (CallFunction|FrontendReactivity)))

(event Custom|Frontend%20Reactivity
  (CallFunction|HitReactivity))

(event Custom|HitReactivity
  (bind _returnvalue (CallFunction|GetIdleVFXComponent))
  (bind _returnvalue_2 (CallFunction|GetIdleVFXComponent))
  (Utilities|FlowControl|Sequence
    (:then_0
      (Rendering|Material|SetScalarParameterValueonMaterials (Variables|Visual|GetWeaponMesh) "ToggleGlow" 1.0)
      (Utilities|FlowControl|Delay 1.0)
      (Rendering|Material|SetScalarParameterValueonMaterials (Variables|Visual|GetWeaponMesh) "ToggleGlow" 0.0)
      (Rendering|Material|SetScalarParameterValueonMaterials (Variables|Visual|GetWeaponMesh) "TimestampB" (Utilities|Time|GetGameTimeinSeconds))
      (Rendering|Material|SetScalarParameterValueonMaterials (Variables|Visual|GetWeaponMesh) "TransitionLength_InB" 1.0))
    (:then_1
      (Effects|Components|ParticleSystem|SetBoolParameter _returnvalue_2 "HasHitThing" true)
      (Components|Activation|Activate _returnvalue_2 false)
      (Utilities|FlowControl|Sequence
        (:then_0
          (bind (_direction _delay) (Timeline|Timeline PlayFromStart)
            (:Update
              (Effects|Components|ParticleSystem|SetFloatParameter _returnvalue "ReactiveCurve" _delay))
            (:Finished
              (Effects|Components|ParticleSystem|SetBoolParameter _returnvalue "HasHitThing" false))))
        (:then_1
          (Timeline|Timeline_0 PlayFromStart
            (:Finished
              (Components|Activation|Deactivate _returnvalue))))))))

(event FortWeapon|EventOnPlayImpactFX
  (params (HitResult :type Hit%20Result%20Structure%20%28by%20ref%29) (ImpactPhysicalSurface :type EPhysicalSurface%20Enum :default "SurfaceType_Default") (SpawnedPSC :type FXSystem%20Component%20Object%20Reference))
  (bind _returnvalue (CallFunction|GetIdleVFXComponent))
  (bind _returnvalue_2 (CallFunction|GetIdleVFXComponent))
  (Variables|Default|SetImpact%20VFX SpawnedPSC)
  (|CallParentFunction HitResult ImpactPhysicalSurface SpawnedPSC)
  (Utilities|FlowControl|Sequence
    (:then_0
      (Rendering|Material|SetScalarParameterValueonMaterials (Variables|Visual|GetWeaponMesh) "ToggleGlow" 1.0)
      (Utilities|FlowControl|Delay 1.0)
      (Rendering|Material|SetScalarParameterValueonMaterials (Variables|Visual|GetWeaponMesh) "ToggleGlow" 0.0)
      (Rendering|Material|SetScalarParameterValueonMaterials (Variables|Visual|GetWeaponMesh) "TimestampB" (Utilities|Time|GetGameTimeinSeconds))
      (Rendering|Material|SetScalarParameterValueonMaterials (Variables|Visual|GetWeaponMesh) "TransitionLength_InB" 1.0))
    (:then_1
      (Effects|Components|ParticleSystem|SetBoolParameter _returnvalue_2 "HasHitThing" true)
      (Components|Activation|Activate _returnvalue_2 false)
      (Utilities|FlowControl|Sequence
        (:then_0
          (Timeline|Timeline PlayFromStart))
        (:then_1
          (Timeline|Timeline_0 PlayFromStart))))))

(event FortWeapon|EventOnEquippedAndAttached
  (|CallParentFunction)
  (Utilities|FlowControl|Delay 0.2)
  (Components|Activation|Deactivate (CallFunction|GetIdleVFXComponent)))

(fn AudioTrigger
  (params (SoundLibTag :type Sound%20Library%20Tag%20Structure))
  (return false))

(fn DamageDealtReaction)

(fn SetTimestepFadeIn)

(fn SetTimestepPanner)

(fn SetTimestepPannerB)

(fn SetTimestepPannerC)

(fn ConstructionScript
  (|CallParentFunction))
