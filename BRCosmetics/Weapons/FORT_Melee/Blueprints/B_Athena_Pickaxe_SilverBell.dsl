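; Blueprint for the Silver Bell pickaxe cosmetic that feeds a running hit count into its impact effects.
; On each impact FX event it runs the parent behavior, writes the counter to the spawned effect's "Counter" integer parameter, then increments it.
; The counter is never reset or wrapped in this graph, so any per-hit variation is left to the effect asset.
; The construction script only calls the parent.
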(event FortWeapon|EventOnPlayImpactFX
  (params (HitResult :type Hit%20Result%20Structure%20%28by%20ref%29) (ImpactPhysicalSurface :type EPhysicalSurface%20Enum :default "SurfaceType_Default") (SpawnedPSC :type FXSystem%20Component%20Object%20Reference))
  (|CallParentFunction HitResult ImpactPhysicalSurface SpawnedPSC)
  (Effects|Components|ParticleSystem|SetIntParameter SpawnedPSC "Counter" (Variables|Default|Getcounter))
  (Math|Integer|IncrementInt (Variables|Default|Getcounter)))

(fn ConstructionScript
  (|CallParentFunction))
