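; Weapon Blueprint for the SolidMist pickaxe cosmetic, managing its idle, swing and impact particle effects.
; Idle FX follow weapon visibility and are also activated on begin play, on end play, and by the anim blueprint's front-end idle event.
; On impact, it toggles body or regular electric discharge emitters by whether a pawn was hit.
; The instigator's current montage section selects the impact FX direction for each swing.
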(event EventBeginPlay
  (|CallParentFunction)
  (Utilities|FlowControl|Sequence
    (:then_0
      (Components|Activation|Activate (CallFunction|GetIdlePSC) false))
    (:then_1
      (bind _asgalileo_ferry_axe_anim_bp (Utilities|Casting|CastToGalileo_Ferry_Axe_AnimBP (Components|SkeletalMesh|GetAnimInstance (Variables|Visual|GetWeaponMesh))))
      (Delegates|BindEventtoFrontEndIdle _asgalileo_ferry_axe_anim_bp (delegate Custom|FrontEndIdle)))))

(event Custom|FrontEndIdle
  (Utilities|FlowControl|DoOnce false
    (:Completed
      (Utilities|FlowControl|Delay 0.3)
      (Components|Activation|Activate (CallFunction|GetIdlePSC) false))))

(event EventEndPlay
  (params (EndPlayReason :type EEndPlayReason%20Enum :default "Destroyed"))
  (|CallParentFunction EndPlayReason)
  (Utilities|FlowControl|DoOnce false
    (:Completed
      (Utilities|FlowControl|Delay 0.3)
      (Components|Activation|Activate (CallFunction|GetIdlePSC) false))))

(event EventOnWeaponVisibilityChanged
  (params (bVisible :type Boolean) (bSetForLocalControllerOnly :type Boolean))
  (if bVisible
    (Components|Activation|Activate (CallFunction|GetIdleVFXComponent) false)
    (else
      (Components|Activation|Deactivate (CallFunction|GetIdlePSC)))))

(event EventOnInitWeaponCosmetics
  (|CallParentFunction)
  (Components|Activation|Deactivate (CallFunction|GetSwingPSC)))

(event FortWeapon|EventOnPlayImpactFX
  (params (HitResult :type Hit%20Result%20Structure%20%28by%20ref%29) (ImpactPhysicalSurface :type EPhysicalSurface%20Enum :default "SurfaceType_Default") (SpawnedPSC :type FXSystem%20Component%20Object%20Reference))
  (bind (_bblockinghit _binitialoverlap _time _distance _location _impactpoint _normal _impactnormal _physmat _hitactor) (Collision|BreakHitResult HitResult))
  (bind (_asfort_pawn _bsuccess) (Utilities|Casting|CastToFortPawn _hitactor))
  (|CallParentFunction HitResult ImpactPhysicalSurface SpawnedPSC)
  (Utilities|FlowControl|Sequence
    (:then_0
      (Effects|Components|ParticleSystem|SetActorParameter SpawnedPSC "BoneSocketActor" _hitactor)
      (if _bsuccess
        (Effects|Components|ParticleSystem|SetEmitterEnable SpawnedPSC "Electric Discharge" false)
        (Effects|Components|ParticleSystem|SetEmitterEnable SpawnedPSC "Electric Discharge_Body" true)
        (else
          (Effects|Components|ParticleSystem|SetEmitterEnable SpawnedPSC "Electric Discharge" true)
          (Effects|Components|ParticleSystem|SetEmitterEnable SpawnedPSC "Electric Discharge_Body" false))))
    (:then_1
      (Variables|Default|SetImpact%20FX%20Index (Utilities|Array|FindItem (Variables|Default|GetMontageSections) (Animation|Montage|MontageGetCurrentSection (Components|SkeletalMesh|GetAnimInstance (Variables|/Script/Engine.Character|GetMesh (Utilities|Casting|CastToFortPawn (Game|GetInstigator)))) (Animation|Montage|GetCurrentActiveMontage (Components|SkeletalMesh|GetAnimInstance (Variables|/Script/Engine.Character|GetMesh (Utilities|Casting|CastToFortPawn (Game|GetInstigator))))))))
      (Effects|Components|ParticleSystem|SetVectorParameter (Variables|Default|GetImpact%20FX) "Direction" (Utilities|Select :Index (Variables|Default|GetImpact%20FX%20Index) :Option%200 (Variables|Default|GetSwing1%20Direction) :Option%201 (Variables|Default|GetSwing2%20Direction) :Option%202 (Variables|Default|GetSwing3%20Direction) :Option%203 (Variables|Default|GetSwing4%20Direction))))))

(fn ConstructionScript
  (|CallParentFunction)
  (Components|Activation|SetAutoActivate (CallFunction|GetIdlePSC) false))
