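; Cosmetic Blueprint for a spinning Battle Royale pickaxe that drives its swing particle effects from the swing montage section.
; On cosmetic init it deactivates the swing particle component, and its construction script turns off auto-activation on the idle one.
; Right and left swings both get a section index via SetupSwingMontageSectionIndex on the instigator and write it to the swing VFX's SwingSectionIndex integer parameter.
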(event EventOnInitWeaponCosmetics
  (|CallParentFunction)
  (Components|Activation|Deactivate (CallFunction|GetSwingPSC)))

(event EventSwingRight
  ;[0 0 0] Swing
  (|CallParentFunction)
  (bind _section_index (Utilities|SetupSwingMontageSectionIndex (Variables|Actor|GetInstigator))
    (:Out
      (Effects|Components|ParticleSystem|SetIntParameter (CallFunction|GetSwingVFXComponent) "SwingSectionIndex" _section_index))))

(event EventSwingLeft
  ;[0 0 0] Swing
  (|CallParentFunction)
  (bind _section_index (Utilities|SetupSwingMontageSectionIndex (Variables|Actor|GetInstigator))
    (:Out
      (Effects|Components|ParticleSystem|SetIntParameter (CallFunction|GetSwingVFXComponent) "SwingSectionIndex" _section_index))))

(fn ConstructionScript
  (|CallParentFunction)
  (Components|Activation|SetAutoActivate (CallFunction|GetIdlePSC) false))
