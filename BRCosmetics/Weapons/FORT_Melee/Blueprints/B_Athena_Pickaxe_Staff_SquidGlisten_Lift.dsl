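; Cosmetic Blueprint for the SquidGlisten Lift staff pickaxe that varies swing VFX by the montage section being played.
; BeginPlay casts the instigator to a Fort pawn, and the construction script turns off auto-activation of the idle particle component.
; On left and right swings, gated to cosmetic-capable contexts, it reads the pawn's active montage section and sets the swing VFX's IsR2 and IsFourth bools for Swing3 and Swing4.
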(event EventSwingRight
  (|CallParentFunction)
  (Networking|CanExecuteCosmeticEvents
    (:True
      (Utilities|FlowControl|Sequence
        (:then_0
          (Effects|Components|ParticleSystem|SetBoolParameter (CallFunction|GetSwingVFXComponent) "IsR2" (== (Animation|Montage|MontageGetCurrentSection (Components|SkeletalMesh|GetAnimInstance (Variables|/Script/Engine.Character|GetMesh (stale EventBeginPlay _asfort_pawn))) (Animation|Montage|GetCurrentActiveMontage (Components|SkeletalMesh|GetAnimInstance (Variables|/Script/Engine.Character|GetMesh (stale EventBeginPlay _asfort_pawn))))) "Swing3")))
        (:then_1
          (Effects|Components|ParticleSystem|SetBoolParameter (CallFunction|GetSwingVFXComponent) "IsFourth" (== (Animation|Montage|MontageGetCurrentSection (Components|SkeletalMesh|GetAnimInstance (Variables|/Script/Engine.Character|GetMesh (stale EventBeginPlay _asfort_pawn))) (Animation|Montage|GetCurrentActiveMontage (Components|SkeletalMesh|GetAnimInstance (Variables|/Script/Engine.Character|GetMesh (stale EventBeginPlay _asfort_pawn))))) "Swing4")))))))

(event EventSwingLeft
  (|CallParentFunction)
  (Networking|CanExecuteCosmeticEvents
    (:True
      (Utilities|FlowControl|Sequence
        (:then_0
          (Effects|Components|ParticleSystem|SetBoolParameter (CallFunction|GetSwingVFXComponent) "IsR2" (== (Animation|Montage|MontageGetCurrentSection (Components|SkeletalMesh|GetAnimInstance (Variables|/Script/Engine.Character|GetMesh (stale EventBeginPlay _asfort_pawn))) (Animation|Montage|GetCurrentActiveMontage (Components|SkeletalMesh|GetAnimInstance (Variables|/Script/Engine.Character|GetMesh (stale EventBeginPlay _asfort_pawn))))) "Swing3")))
        (:then_1
          (Effects|Components|ParticleSystem|SetBoolParameter (CallFunction|GetSwingVFXComponent) "IsFourth" (== (Animation|Montage|MontageGetCurrentSection (Components|SkeletalMesh|GetAnimInstance (Variables|/Script/Engine.Character|GetMesh (stale EventBeginPlay _asfort_pawn))) (Animation|Montage|GetCurrentActiveMontage (Components|SkeletalMesh|GetAnimInstance (Variables|/Script/Engine.Character|GetMesh (stale EventBeginPlay _asfort_pawn))))) "Swing4")))))))

(event EventBeginPlay
  (bind _asfort_pawn (Utilities|Casting|CastToFortPawn (Game|GetInstigator))))

(fn ConstructionScript
  (|CallParentFunction)
  (Components|Activation|SetAutoActivate (CallFunction|GetIdlePSC) false))
