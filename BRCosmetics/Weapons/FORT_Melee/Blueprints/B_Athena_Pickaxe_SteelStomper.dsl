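; Blueprint for the Steel Stomper pickaxe cosmetic that updates its idle particle effect when weapon visibility changes.
; The visibility event calls its parent, then passes the visible flag through a server-named custom event to a client-named one.
; When visible, the client event sets the idle VFX's bIsSwinging parameter to true.
; Despite the hide-on-swap naming, nothing happens to the effect when the weapon becomes hidden.
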(event EventOnWeaponVisibilityChanged
  (params (bVisible :type Boolean) (bSetForLocalControllerOnly :type Boolean))
  (|CallParentFunction bVisible false)
  (CallFunction|FXHideonSwapServer bVisible))

(event Custom|FX_HideOnSwapServer
  (params (IsVisible :type Boolean))
  (CallFunction|FXHideonSwapClient IsVisible))

(event Custom|FX_HideOnSwapClient
  (params (IsVisible :type Boolean))
  (if IsVisible
    (Effects|Components|ParticleSystem|SetBoolParameter (CallFunction|GetIdleVFXComponent) "bIsSwinging" true)))

(fn ConstructionScript
  (|CallParentFunction))
