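; Cosmetic Blueprint for the TameEagle pickaxe, driving its emissive glow and a Niagara energy trail on each swing.
; Left and right swings spawn the trail, using a computed impact point in frontend preview and line traces otherwise, though trace results never gate spawning.
; Impact FX feeds the real hit location into Niagara.
; When the glow timeline ends, a returning trail spawns and a camera-distance check gates the return glow; unequip clears effects.
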(event Custom|Glow
  (bind (_direction _meshglow _timer _trailglowamount) (Timeline|TL_PickaxeGlow PlayFromStart)
    (:Update
      (Rendering|Material|SetScalarParameterValueonMaterials (Variables|Visual|GetWeaponMesh) "Emissive_Intensity" (* _meshglow 5.0))
      (Effects|Components|ParticleSystem|SetFloatParameter (CallFunction|GetSwingVFXComponent) "GlowAmount" _trailglowamount))
    (:Finished
      (Variables|Default|SetFirstTime true)
      (CallFunction|ReturningEvent))))

(event Custom|ReturnGlow
  ; DistanceCheck. If this is changed, make sure to adjust it in the NiagaraSystem aswell
  (if (> (Math|Vector|Distance(Vector) (Camera|GetCameraLocation (Game|GetPlayerCameraManager 0)) (Variables|Default|GetLastImpactLocation)) 8e+02)
    (Variables|Default|SetLastImpactLocation "0, 0, 0")
    (Variables|Default|SetFirstTime true)
    (else
      (bind (_direction _newtrack_0) (Timeline|TL_Return PlayFromStart)
        (:Update
          (Rendering|Material|SetScalarParameterValueonMaterials (Variables|Visual|GetWeaponMesh) "Emissive_Intensity" (* _newtrack_0 5.0)))))))

(event EventSwingLeft
  (bind _tl_pickaxeglow (Variables|B_Athena_Pickaxe_TameEagle|GetTL_PickaxeGlow))
  ; Swing Left/Right
  (|CallParentFunction)
  (Utilities|FlowControl|Sequence
    (:then_0
      _)
    (:then_1
      (Utilities|FlowControl|RetriggerableDelay 0.1)))
  (Utilities|FlowControl|DoOnce false
    (:Completed
      (Variables|Default|SetRightSwing false)
      (bind _asfort_player_pawn (Utilities|Casting|CastToFortPlayerPawn (Game|GetInstigator)))
      (Variables|Default|SetPlayer _asfort_player_pawn)
      (if (or (Components|Timeline|IsPlaying _tl_pickaxeglow) (Components|Timeline|IsPlaying (Variables|B_Athena_Pickaxe_TameEagle|GetTL_Return)))
        (Components|Timeline|Stop _tl_pickaxeglow))
      (if (Reactive|Frontend|IsCosmeticObjectDisplayedasPreview 0)
        (CallFunction|SpawnFairyTrail)
        (Variables|Default|SetLastImpactLocation (CallFunction|GetImpactLocation))
        (else
          ;; Using a timer incase the player turns mid swing, since it can still trigger an impact
          (CallFunction|LineTraceforImpact)
          (bind _returnvalue (Utilities|Time|SetTimerbyEvent (EventDispatchers|CreateEvent :!event Custom|LineTraceForImpact) 0.05 true))
          (Variables|Default|SetTimerHandle _returnvalue)
          (Utilities|FlowControl|Delay 0.2)
          (Utilities|Time|ClearandInvalidateTimerbyHandle (Variables|Default|GetTimerHandle)))))))

(event EventSwingRight
  (bind _tl_pickaxeglow (Variables|B_Athena_Pickaxe_TameEagle|GetTL_PickaxeGlow))
  ; Swing Left/Right
  (|CallParentFunction)
  (Utilities|FlowControl|Sequence
    (:then_0
      _)
    (:then_1
      (Utilities|FlowControl|RetriggerableDelay 0.1)))
  (Utilities|FlowControl|DoOnce false
    (:Completed
      (Variables|Default|SetRightSwing true)
      (bind _asfort_player_pawn (Utilities|Casting|CastToFortPlayerPawn (Game|GetInstigator)))
      (Variables|Default|SetPlayer _asfort_player_pawn)
      (if (or (Components|Timeline|IsPlaying _tl_pickaxeglow) (Components|Timeline|IsPlaying (Variables|B_Athena_Pickaxe_TameEagle|GetTL_Return)))
        (Components|Timeline|Stop _tl_pickaxeglow))
      (if (Reactive|Frontend|IsCosmeticObjectDisplayedasPreview 0)
        (CallFunction|SpawnFairyTrail)
        (Variables|Default|SetLastImpactLocation (CallFunction|GetImpactLocation))
        (else
          ;; Using a timer incase the player turns mid swing, since it can still trigger an impact
          (CallFunction|LineTraceforImpact)
          (bind _returnvalue (Utilities|Time|SetTimerbyEvent (EventDispatchers|CreateEvent :!event Custom|LineTraceForImpact) 0.05 true))
          (Variables|Default|SetTimerHandle _returnvalue)
          (Utilities|FlowControl|Delay 0.2)
          (Utilities|Time|ClearandInvalidateTimerbyHandle (Variables|Default|GetTimerHandle)))))))

(event Custom|ReturningEvent
  (bind _returnvalue (Reactive|Frontend|IsCosmeticObjectDisplayedasPreview 0))
  (bind _ns_energytrail (Variables|Default|GetNS_EnergyTrail))
  (bind _ns_energytrail_2 (Variables|Default|GetNS_EnergyTrail))
  (bind _ns_energytrail_3 (Variables|Default|GetNS_EnergyTrail))
  (bind _ns_energytrail_4 (Variables|Default|GetNS_EnergyTrail))
  ; Actual behavior logic when you hit something
  (Variables|Default|SetReturning true)
  (Utilities|FlowControl|Sequence
    (:then_0
      (Utilities|FlowControl|Delay 0.3)
      (if (Variables|Default|GetReturning)
        (CallFunction|ReturnGlow)))
    (:then_1
      (Utilities|IsValid _ns_energytrail_2
        (:Is%20Valid
          (Components|DestroyComponent _ns_energytrail_2))
        (:Is%20Not%20Valid
          _))
      (bind _returnvalue_2 (Niagara|SpawnSystemAttached "/BRCosmetics/Weapons/FORT_Melee/Pickaxe_TameEagle/FX/NS_Pickaxe_TameEagle_EnergyTrail.NS_Pickaxe_TameEagle_EnergyTrail" (Variables|Visual|GetWeaponMesh) "root"))
      (bind _output_get (Variables|Default|SetNS_EnergyTrail _returnvalue_2))
      (Niagara|SetNiagaraSkeletalMeshComponent _output_get "Character" (Variables|/Script/Engine.Character|GetMesh (Variables|Default|GetPlayer)))
      (Niagara|SetNiagaraVariable(Bool) _ns_energytrail_4 "bIsFrontEnd" _returnvalue)
      (Niagara|SetNiagaraVariable(Bool) _ns_energytrail_4 "bIsFrontEndPreview" _returnvalue)
      (if (Variables|Default|GetReturning)
        (Niagara|SetNiagaraVariable(Bool) _ns_energytrail_3 "Returning" true)
        (Niagara|SetNiagaraVariable(Vector3) (Variables|Default|GetNS_EnergyTrail) "ImpactLocation" (Variables|Default|GetLastImpactLocation))
        (Variables|Default|SetFirstTime true)
        (else
          (Niagara|SetNiagaraVariable(Bool) _ns_energytrail_3 "RightSwing" (Variables|Default|GetRightSwing))
          (Niagara|SetNiagaraVariable(Bool) _ns_energytrail_3 "FirstTime" (Variables|Default|GetFirstTime))
          (if (Variables|Default|GetFirstTime)
            (Variables|Default|SetFirstTime false)
            (Niagara|SetNiagaraVariable(Bool) _ns_energytrail "FoundImpact" false)
            (Niagara|SetNiagaraVariable(Vector3) _ns_energytrail "ImpactLocation" (select (Reactive|Frontend|IsCosmeticObjectDisplayedasPreview 0) (CallFunction|GetImpactLocation) (Transformation|GetSocketLocation (Variables|Visual|GetWeaponMesh) "FX_Trail_1")))
            (else
              (Niagara|SetNiagaraVariable(Vector3) (Variables|Default|GetNS_EnergyTrail) "ImpactLocation" (Variables|Default|GetLastImpactLocation)))))))))

(event Custom|LineTraceForImpact
  (bind _player (Variables|Default|GetPlayer))
  (bind _array (Utilities|Array|MakeArray _player))
  ; LineTrace Checks
  (bind (_outhit _returnvalue) (Collision|LineTraceByChannel (Utilities|Operators|Add (Camera|GetCameraLocation (Game|GetPlayerCameraManager 0)) "0.000000,0.000000,5.000000" (* (Math|Vector|GetForwardVector (Pawn|GetControlRotation (Pawn|GetController _player) :!class /Script/Engine.Controller)) 2.5e+02)) (+ (Utilities|Operators|Add (Camera|GetCameraLocation (Game|GetPlayerCameraManager 0)) "0.000000,0.000000,5.000000" (* (Math|Vector|GetForwardVector (Pawn|GetControlRotation (Pawn|GetController _player) :!class /Script/Engine.Controller)) 2.5e+02)) (* (Math|Vector|GetForwardVector (Pawn|GetControlRotation (Pawn|GetController _player) :!class /Script/Engine.Controller)) 312.0)) "TraceTypeQuery3" false _array))
  (if (not _returnvalue)
    (bind (_outhit_2 _returnvalue_2) (Collision|LineTraceByChannel (+ (Transformation|GetWorldLocation (Variables|/Script/Engine.Character|GetMesh (Variables|Default|GetPlayer))) "0.000000,0.000000,90.000000") (+ (+ (Transformation|GetWorldLocation (Variables|/Script/Engine.Character|GetMesh (Variables|Default|GetPlayer))) "0.000000,0.000000,90.000000") (* (Math|Vector|GetRightVector (Transformation|GetWorldRotation (Variables|/Script/Engine.Character|GetMesh (Variables|Default|GetPlayer)))) 3.2e+02)) "TraceTypeQuery3" false _array "None" true "(R=1.000000,G=0.031562,B=0.599340,A=1.000000)" "(R=0.000000,G=0.158763,B=1.000000,A=1.000000)"))
    (if _returnvalue_2
      _))
  (CallFunction|SpawnFairyTrail)
  (Utilities|FlowControl|DelayUntilNextFrame)
  (Utilities|Time|ClearandInvalidateTimerbyHandle (Variables|Default|GetTimerHandle)))

(event Custom|Spawn%20FairyTrail
  (bind _returnvalue (Reactive|Frontend|IsCosmeticObjectDisplayedasPreview 0))
  (bind _ns_energytrail (Variables|Default|GetNS_EnergyTrail))
  (bind _ns_energytrail_2 (Variables|Default|GetNS_EnergyTrail))
  (bind _ns_energytrail_3 (Variables|Default|GetNS_EnergyTrail))
  (bind _ns_energytrail_4 (Variables|Default|GetNS_EnergyTrail))
  ; Actual behavior logic when you hit something
  (Variables|Default|SetReturning false)
  (CallFunction|Glow)
  (Utilities|IsValid _ns_energytrail_2
    (:Is%20Valid
      (Components|DestroyComponent _ns_energytrail_2))
    (:Is%20Not%20Valid
      _))
  (bind _returnvalue_2 (Niagara|SpawnSystemAttached "/BRCosmetics/Weapons/FORT_Melee/Pickaxe_TameEagle/FX/NS_Pickaxe_TameEagle_EnergyTrail.NS_Pickaxe_TameEagle_EnergyTrail" (Variables|Visual|GetWeaponMesh) "root"))
  (bind _output_get (Variables|Default|SetNS_EnergyTrail _returnvalue_2))
  (Niagara|SetNiagaraSkeletalMeshComponent _output_get "Character" (Variables|/Script/Engine.Character|GetMesh (Variables|Default|GetPlayer)))
  (Niagara|SetNiagaraVariable(Bool) _ns_energytrail_4 "bIsFrontEnd" _returnvalue)
  (Niagara|SetNiagaraVariable(Bool) _ns_energytrail_4 "bIsFrontEndPreview" _returnvalue)
  (if (Variables|Default|GetReturning)
    (Niagara|SetNiagaraVariable(Bool) _ns_energytrail_3 "Returning" true)
    (Niagara|SetNiagaraVariable(Vector3) (Variables|Default|GetNS_EnergyTrail) "ImpactLocation" (Variables|Default|GetLastImpactLocation))
    (Variables|Default|SetFirstTime true)
    (else
      (Niagara|SetNiagaraVariable(Bool) _ns_energytrail_3 "RightSwing" (Variables|Default|GetRightSwing))
      (Niagara|SetNiagaraVariable(Bool) _ns_energytrail_3 "FirstTime" (Variables|Default|GetFirstTime))
      (if (Variables|Default|GetFirstTime)
        (Variables|Default|SetFirstTime false)
        (Niagara|SetNiagaraVariable(Bool) _ns_energytrail "FoundImpact" false)
        (Niagara|SetNiagaraVariable(Vector3) _ns_energytrail "ImpactLocation" (select (Reactive|Frontend|IsCosmeticObjectDisplayedasPreview 0) (CallFunction|GetImpactLocation) (Transformation|GetSocketLocation (Variables|Visual|GetWeaponMesh) "FX_Trail_1")))
        (else
          (Niagara|SetNiagaraVariable(Vector3) (Variables|Default|GetNS_EnergyTrail) "ImpactLocation" (Variables|Default|GetLastImpactLocation)))))))

(event FortWeapon|EventOnPlayImpactFX
  (params (HitResult :type Hit%20Result%20Structure%20%28by%20ref%29) (ImpactPhysicalSurface :type EPhysicalSurface%20Enum :default "SurfaceType_Default") (SpawnedPSC :type FXSystem%20Component%20Object%20Reference))
  (bind (_bblockinghit _binitialoverlap _time _distance _location) (Collision|BreakHitResult HitResult))
  (bind _ns_energytrail (Variables|Default|GetNS_EnergyTrail))
  ; Use the ImpactLocation as soon as its available
  (|CallParentFunction HitResult ImpactPhysicalSurface 0)
  (Niagara|SetNiagaraVariable(Bool) _ns_energytrail "FoundImpact" true)
  (bind _output_get (Variables|Default|SetLastImpactLocation _location))
  (Niagara|SetNiagaraVariable(Vector3) _ns_energytrail "ImpactLocation" _output_get))

(event FortWeapon|EventOnLocalUnEquip
  ; Kill any logic the moment the weapon is unequipped
  (|CallParentFunction)
  (Components|DestroyComponent (Variables|Default|GetNS_EnergyTrail))
  (Components|Timeline|Stop (Variables|B_Athena_Pickaxe_TameEagle|GetTL_PickaxeGlow))
  (Variables|Default|SetFirstTime true)
  (Variables|Default|SetLastImpactLocation "0, 0, 0")
  (Rendering|Material|SetScalarParameterValueonMaterials (Variables|Visual|GetWeaponMesh) "Emissive_Intensity" 0.0))

(fn GetImpactLocation
  (return (Utilities|Operators|Add :A (+ (Transformation|GetWorldLocation (Variables|/Script/Engine.Character|GetMesh (Variables|Default|GetPlayer))) "0.000000,0.000000,110.000000") :B (Math|Vector|RotateVector (+ (Variables|/Script/FortniteGame.AthenaPickaxeItemDefinition|GetPickaxeImpactFXOnPawnPreviewOffset (Creative|ItemData|GetItemDefinition (Utilities|Struct|MakeItemVariantHandle "/BRCosmetics/Athena/Items/Cosmetics/Pickaxes/Pickaxe_TameEagle.Pickaxe_TameEagle") "/Script/FortniteGame.AthenaPickaxeItemDefinition")) "135.000000,0.000000,0.000000") (Pawn|GetBaseAimRotation (Variables|Default|GetPlayer))))))

(fn ConstructionScript
  (|CallParentFunction))
