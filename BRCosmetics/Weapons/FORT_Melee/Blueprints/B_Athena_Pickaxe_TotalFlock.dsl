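; Cosmetic Blueprint for the TotalFlock pickaxe, driving reactive material and VFX responses to swings and weakpoint hits.
; Weakpoint hits stamp game time into two alternating material parameters, set a long ramp-down and briefly flag the idle VFX.
; Swings pass montage section to swing VFX and trigger shorter material pulses unless a weakpoint reaction is active.
; In frontend display, a looping timer replays the weakpoint reaction as a preview.
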(event EventOnInitWeaponCosmetics
  (if (Reactive|Frontend|IsCosmeticObjectDisplayedinFrontendContext 0)
    (Rendering|Material|SetScalarParameterValueonMaterials (Variables|Visual|GetWeaponMesh) "ReactivityRampDownDuration" 8.0)
    (CallFunction|FrontendReactivity)
    (Utilities|FlowControl|Delay 4.0)
    (Utilities|Time|SetTimerbyEvent (delegate Custom|Frontend%20Reactivity) 9.0 true)))

(event Custom|Frontend%20Reactivity
  (CallFunction|HitReactivity))

(event EventOnPlayWeaponFireFX
  (params (bPersistentFire :type Boolean) (bSecondaryFire :type Boolean))
  (bind _returnvalue (CallFunction|GetSwingVFXComponent))
  (bind _weaponmesh (Variables|Visual|GetWeaponMesh))
  (bind _section_index (Utilities|SetupSwingMontageSectionIndex (Variables|Actor|GetInstigator))
    (:Out
      (Effects|Components|ParticleSystem|SetIntParameter _returnvalue "SwingSectionIndex" _section_index)
      (if (Variables|Default|GetHasFinishedWeakpoint?)
        (Rendering|Material|SetScalarParameterValueonMaterials _weaponmesh "ReactivityRampDownDuration" 1.5)
        (Math|Integer|IncrementInt (Variables|Default|GetHit%20Counter)
          (:%20%20
            (Utilities|DoOncePerFrame
              (:%20%20
                (Rendering|Material|SetScalarParameterValueonMaterials (Variables|Visual|GetWeaponMesh) (Utilities|String|StringToName (Utilities|String|Append "DmgReactivityStartTime_" (Utilities|String|ToString(Integer) (Math|Integer|%25%28Integer%29 (Variables|Default|GetHit%20Counter) 2)))) (Utilities|Time|GetGameTimeinSeconds))))))
        (else
          (Rendering|Material|SetScalarParameterValueonMaterials _weaponmesh "ReactivityRampDownDuration" 8.0))))
    (:Cast%20Failed%20Out
      (Effects|Components|ParticleSystem|SetIntParameter _returnvalue "SwingSectionIndex" -1))))

(event Custom|Hit%20Reactivity
  (bind _returnvalue (CallFunction|GetIdleVFXComponent))
  ;[0 1 0.9093 0.5] WEAKPOINT IMPACT
  (Utilities|FlowControl|Sequence
    (:then_0
      (Rendering|Material|SetScalarParameterValueonMaterials (Variables|Visual|GetWeaponMesh) "ReactivityRampDownDuration" 8.0)
      (Math|Integer|IncrementInt (Variables|Default|GetHit%20Counter)
        (:%20%20
          (Utilities|DoOncePerFrame
            (:%20%20
              (Rendering|Material|SetScalarParameterValueonMaterials (Variables|Visual|GetWeaponMesh) (Utilities|String|StringToName (Utilities|String|Append "DmgReactivityStartTime_" (Utilities|String|ToString(Integer) (Math|Integer|%25%28Integer%29 (Variables|Default|GetHit%20Counter) 2)))) (Utilities|Time|GetGameTimeinSeconds))
              (Variables|Default|SetHasFinishedWeakpoint? false)
              (Utilities|FlowControl|Delay 3.0)
              (Variables|Default|SetHasFinishedWeakpoint? true))))))
    (:then_1
      (Effects|Components|ParticleSystem|SetBoolParameter _returnvalue "bWeakpointHit" true)
      (Utilities|FlowControl|Delay 3.0)
      (Effects|Components|ParticleSystem|SetBoolParameter _returnvalue "bWeakpointHit" false))))

(event EventOnPlayWeakpointHitFX
  (bind _returnvalue (CallFunction|GetIdleVFXComponent))
  (|CallParentFunction)
  ;[0 1 0.9093 0.5] WEAKPOINT IMPACT
  (Utilities|FlowControl|Sequence
    (:then_0
      (Rendering|Material|SetScalarParameterValueonMaterials (Variables|Visual|GetWeaponMesh) "ReactivityRampDownDuration" 8.0)
      (Math|Integer|IncrementInt (Variables|Default|GetHit%20Counter)
        (:%20%20
          (Utilities|DoOncePerFrame
            (:%20%20
              (Rendering|Material|SetScalarParameterValueonMaterials (Variables|Visual|GetWeaponMesh) (Utilities|String|StringToName (Utilities|String|Append "DmgReactivityStartTime_" (Utilities|String|ToString(Integer) (Math|Integer|%25%28Integer%29 (Variables|Default|GetHit%20Counter) 2)))) (Utilities|Time|GetGameTimeinSeconds))
              (Variables|Default|SetHasFinishedWeakpoint? false)
              (Utilities|FlowControl|Delay 3.0)
              (Variables|Default|SetHasFinishedWeakpoint? true))))))
    (:then_1
      (Effects|Components|ParticleSystem|SetBoolParameter _returnvalue "bWeakpointHit" true)
      (Utilities|FlowControl|Delay 3.0)
      (Effects|Components|ParticleSystem|SetBoolParameter _returnvalue "bWeakpointHit" false))))

(event EventSwingRight
  ; JAB EFFECT DISABLED
  (|CallParentFunction))

(event EventSwingLeft
  ; JAB EFFECT DISABLED
  (|CallParentFunction))

(fn ConstructionScript
  (|CallParentFunction)
  (Components|Activation|SetAutoActivate (CallFunction|GetIdlePSC) false))
