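; B_Athena_Pickaxe_TrackWaver is a cosmetic pickaxe blueprint that makes the TrackWaver pickaxe react visually to swings.
; Left and right swings both call the parent event, then play the SwingGlow timeline from the start.
; Each timeline update feeds the glow value into a scalar parameter on the weapon mesh's dynamic material and a float parameter on the idle VFX particle component.
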(event EventSwingLeft
  (|CallParentFunction)
  (bind (_direction _swingglow) (Timeline|SwingGlow PlayFromStart)
    (:Update
      (Rendering|Material|SetScalarParameterValue (Class|BTechArtGlobalLibrary|GetDynamicMaterialInstance (Variables|Visual|GetWeaponMesh)) "TW_Reactvity" _swingglow :!class /Script/Engine.MaterialInstanceDynamic)
      (Effects|Components|ParticleSystem|SetFloatParameter (CallFunction|GetIdleVFXComponent) "TW_FXReactvity" _swingglow))))

(event EventSwingRight
  (|CallParentFunction)
  (Timeline|SwingGlow PlayFromStart))

(fn ConstructionScript
  (|CallParentFunction))
