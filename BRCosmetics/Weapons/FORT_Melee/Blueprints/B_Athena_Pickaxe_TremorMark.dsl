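; Tremor Mark pickaxe weapon Blueprint that makes its idle and swing VFX react to hits.
; On equip, it activates the idle effect after a 0.2-second delay.
; On impact, it flags HasHitThing on the swing effect, deactivates idle and restarts a timeline driving the idle effect's ReactiveCurve.
; When the timeline finishes, it clears HasHitThing on the idle effect, not the swing one, and reactivates idle.
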(event EventBeginPlay)

(event EventTick
  (params (DeltaSeconds :type Float%20%28single-precision%29)))

(event FortWeapon|EventOnPlayImpactFX
  (params (HitResult :type Hit%20Result%20Structure%20%28by%20ref%29) (ImpactPhysicalSurface :type EPhysicalSurface%20Enum :default "SurfaceType_Default") (SpawnedPSC :type FXSystem%20Component%20Object%20Reference))
  (bind _returnvalue (CallFunction|GetIdleVFXComponent))
  (|CallParentFunction HitResult ImpactPhysicalSurface SpawnedPSC)
  (Effects|Components|ParticleSystem|SetBoolParameter (CallFunction|GetSwingVFXComponent) "HasHitThing" true)
  (Components|Activation|Deactivate _returnvalue)
  (bind (_direction _fade) (Timeline|Timeline PlayFromStart)
    (:Update
      (Effects|Components|ParticleSystem|SetFloatParameter _returnvalue "ReactiveCurve" _fade))
    (:Finished
      (Effects|Components|ParticleSystem|SetBoolParameter _returnvalue "HasHitThing" false)
      (Components|Activation|Activate _returnvalue false))))

(event FortWeapon|EventOnEquippedAndAttached
  (|CallParentFunction)
  (Utilities|FlowControl|Delay 0.2)
  (Components|Activation|Activate (CallFunction|GetIdleVFXComponent) false))

(fn ConstructionScript
  (|CallParentFunction))
