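; WeaveHarbor cosmetic pickaxe Blueprint that reacts to impacts on listed physical surfaces with glow, particle and sound effects.
; The impact event routes matching hits through FX_Server to FX_Client, while frontend or preview displays react to every impact.
; The shared update flags impact and idle particles as weak-point hits, stamps game time into mesh materials, and runs a timeline fading the idle effect.
; The impact sound spawns in both FX_Client and the shared update.
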(event Custom|FrontEndReactivity
  (Rendering|Material|SetScalarParameterValueonMaterials (Variables|Visual|GetWeaponMesh) "BP_GetTime" (Utilities|Time|GetGameTimeinSeconds))
  (CallFunction|FXClient))

(event EventBeginPlay)

(event EventTick
  (params (DeltaSeconds :type Float%20%28single-precision%29)))

(event FortWeapon|EventOnPlayImpactFX
  (params (HitResult :type Hit%20Result%20Structure%20%28by%20ref%29) (ImpactPhysicalSurface :type EPhysicalSurface%20Enum :default "SurfaceType_Default") (SpawnedPSC :type FXSystem%20Component%20Object%20Reference))
  (|CallParentFunction HitResult ImpactPhysicalSurface SpawnedPSC)
  (Variables|Default|SetSpawnedPSC SpawnedPSC)
  (if (or (Reactive|Frontend|IsCosmeticObjectDisplayedinFrontendContext 0) (Reactive|Frontend|IsCosmeticObjectDisplayedasPreview 0))
    (CallFunction|FrontEndReactivity)
    (elif (Utilities|Array|ContainsItem (Variables|Default|GetImpactPhysicalSurface) ImpactPhysicalSurface)
      (Effects|Components|ParticleSystem|SetBoolParameter (Variables|Default|GetSpawnedPSC) "bWeakPointHit" true)
      (CallFunction|FXServer))))

(event Custom|FX_Server
  (CallFunction|FXClient))

(event Custom|FX_Client
  (CallFunction|UpdateMaterialandTriggerFX)
  (Networking|CanExecuteCosmeticEvents
    (:True
      (Audio|SpawnSoundAttached "/BRCosmetics/Sounds/PickAxes/WeaveHarbor/PickaxeImpactEnemy_Reactive_WeaveHarbor_Cue.PickaxeImpactEnemy_Reactive_WeaveHarbor_Cue" (Variables|Visual|GetWeaponMesh)))))

(event Custom|UpdateMaterialAndTriggerFX
  (Utilities|FlowControl|Sequence
    (:then_0
      (Effects|Components|ParticleSystem|SetBoolParameter (CallFunction|GetIdleVFXComponent) "bWeakPointHit" true)
      (Effects|Components|ParticleSystem|SetBoolParameter (Variables|Default|GetSpawnedPSC) "bWeakPointHit" true)
      (Rendering|Material|SetScalarParameterValueonMaterials (Variables|Visual|GetWeaponMesh) "BP_GetTime" (Utilities|Time|GetGameTimeinSeconds))
      (bind (_direction _weakpointhitfade) (Timeline|WeaveHarborPickaxeTimeline PlayFromStart)
        (:Update
          (Effects|Components|ParticleSystem|SetFloatParameter (CallFunction|GetIdleVFXComponent) "WeakPointHitFade" _weakpointhitfade))
        (:Finished
          (Effects|Components|ParticleSystem|SetBoolParameter (CallFunction|GetIdleVFXComponent) "bWeakPointHit" false))))
    (:then_1
      ;[1 0.2552 0.2552] Material Reactivity
      (Rendering|Material|SetScalarParameterValueonMaterials (Variables|Visual|GetWeaponMesh) "BP_GetTime" (Utilities|Time|GetGameTimeinSeconds))
      (Audio|SpawnSoundAttached "/BRCosmetics/Sounds/PickAxes/WeaveHarbor/PickaxeImpactEnemy_Reactive_WeaveHarbor_Cue.PickaxeImpactEnemy_Reactive_WeaveHarbor_Cue" (Variables|Visual|GetWeaponMesh)))))

(fn ConstructionScript
  (|CallParentFunction))
