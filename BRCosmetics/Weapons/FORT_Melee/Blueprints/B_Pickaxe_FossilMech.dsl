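; Fossil Mech pickaxe Blueprint that drives a timed glow-style material and particle effect on impacts.
; Each impact passes an incrementing counter to the spawned FX, and frontend display also triggers the effect update.
; Weak-point surface hits update directly, while other hits relay through server and client events to the same update.
; That update stamps game time into weapon material and idle, swing and impact FX parameters, avoiding tick for the commented six-second window.
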(event EventBeginPlay)

(event EventTick
  (params (DeltaSeconds :type Float%20%28single-precision%29)))

(event FortWeapon|EventOnPlayImpactFX
  (params (HitResult :type Hit%20Result%20Structure%20%28by%20ref%29) (ImpactPhysicalSurface :type EPhysicalSurface%20Enum :default "SurfaceType_Default") (SpawnedPSC :type FXSystem%20Component%20Object%20Reference))
  (Variables|Default|SetSpawned%20PSC SpawnedPSC)
  (|CallParentFunction HitResult ImpactPhysicalSurface SpawnedPSC)
  (Effects|Components|ParticleSystem|SetIntParameter SpawnedPSC "Counter" (Variables|Default|GetCounter))
  (Variables|Default|SetCounter (+ (Variables|Default|GetCounter) 1))
  (Utilities|FlowControl|Sequence
    (:then_0
      (if (Reactive|Frontend|IsCosmeticObjectDisplayedinFrontendContext self)
        (CallFunction|UpdateMaterialProgressEvent)))
    (:then_1
      (bind _output_get (Variables|Default|SetCurrent%20Impact%20Physical%20Surface%200 ImpactPhysicalSurface))
      (Variables|Default|SetIs%20Hit%20WeakPoint_0 (Utilities|Array|ContainsItem (Variables|Default|GetWeakPoint%20Physical%20Surface%20Array_0) _output_get))
      (if (Variables|Default|GetIs%20Hit%20WeakPoint_0)
        (CallFunction|UpdateMaterialProgressEvent)
        (CallFunction|FossilMechFXEventServer "SurfaceType_Default")))))

(event Custom|FossilMech_FX_Event_Server
  (params (Impact%20Physical%20Surface :type EPhysicalSurface%20Enum :default "SurfaceType_Default"))
  (CallFunction|FossilMechFXEventClient Impact%20Physical%20Surface))

(event Custom|FossilMech_FX_Event_Client
  (params (Impact%20Physical%20Surface :type EPhysicalSurface%20Enum :default "SurfaceType_Default"))
  (CallFunction|UpdateMaterialProgressEvent))

(event Custom|Update_Material_Progress_Event
  ;[0.1 0.1 0.1] Resets the tickless counter
  (Rendering|Material|SetScalarParameterValueonMaterials (Variables|Visual|GetWeaponMesh) "Weak_End_Time" (Utilities|Time|GetGameTimeinSeconds)) ; Weak Lasts for 6 Seconds
  (Effects|Components|ParticleSystem|SetFloatParameter (CallFunction|GetIdleVFXComponent) "WeakPointTime" (Utilities|Time|GetGameTimeinSeconds))
  (Effects|Components|ParticleSystem|SetFloatParameter (Variables|Default|GetSpawned%20PSC) "WeakPointTime" (Utilities|Time|GetGameTimeinSeconds))
  (Effects|Components|ParticleSystem|SetFloatParameter (CallFunction|GetSwingVFXComponent) "WeakPointTime" (Utilities|Time|GetGameTimeinSeconds)))

(fn ConstructionScript
  (|CallParentFunction))
