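# Unlit translucent particle material for the Dark Eagle melee weapon's storm globe, faking a swirling cloud sphere on a sprite.
# A sphere normal map combines with two rotating, panning cloud layers for two-tone smoke, rim light, sun/moon highlight and emissive lightning flashes.
# Dynamic parameters randomize panning, offset lightning and timestamp an eased spin burst.
# Sphere masks shape opacity, with optional view-angle fade; Mobile drops the highlight and substitutes cheaper mask-based shading.
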Material /BRCosmetics/Weapons/FORT_Melee/Effects/M_Dark_Eagle_StormGlobe_Sprite
graph {
 properties {
  bUsedWithNiagaraMeshParticles=true
  bUsedWithNiagaraSprites=true
  bUsedWithParticleSprites=true
  blendMode=BLEND_Translucent
  pixelDepthOffsetMode=PDOM_Legacy
  shadingModel=MSM_Unlit
  usedShadingModels=Unlit
 }
 nodes [
  Constant() {r=1}
  n1 = CollectionParameter() {
   collection=&FortniteMaterialParameters
   parameterName=SunAndMoonModelDirectionalVector
  }
  n2 = TextureSample() {samplerType=SAMPLERTYPE_Normal, texture=&T_Sphere_N}
  n3 = DotProduct(A=ComponentMask(None=@n1) {b=true, g=true, r=true}, B=ReflectionVectorWS(CustomWorldNormal=Transform(None=@n2.RGB)))
  n4 = Multiply(
   A=Power(Base=Clamp(None=@n3), Exp=ScalarParameter() {defaultValue=70, parameterName=Spec Highlight Power}) {constExponent=20}
   B=ScalarParameter() {defaultValue=0.5, parameterName=Spec Intense}
  )
  n5 = Constant()
  n6 = ParticleColor()
  n7 = VectorParameter() {parameterName=SmokeColor1, defaultValue=<rgba 0.0484 0.0653 0.1615 0>}
  n8 = DynamicParameter() {paramNames [Random,PanSpeed,TimeStamp,Offset]}
  n9 = Subtract(A=Time() {period=5}, B=@n8.TimeStamp)
  n10 = ScalarParameter() {defaultValue=3, group=Timestamp, parameterName=SpinDuration}
  n11 = MaterialFunctionCall!(In=Clamp(None=Divide(A=@n9, B=@n10))) {materialFunction=/Game/Effects/Fort_Effects/Materials/_Functions/MF_Smoothstep_EaseOut}
  n12 = ScalarParameter() {defaultValue=-3, parameterName=SpinAmount}
  n13 = Time()
  n14 = Multiply(A=ScalarParameter() {defaultValue=0.5, parameterName=RotationSpeed}, B=Add(A=Multiply(A=@n11.Result, B=@n12) {constB=-18}, B=@n13))
  n15 = Multiply(A=@n11.Result) {constB=-2}
  n16 = Add(A=Multiply(A=Add(A=@n15, B=@n13), B=@n8.PanSpeed), B=@n8.Random)
  n17 = Panner(Coordinate=Rotator(Coordinate=TextureCoordinate() {uTiling=0.6, vTiling=0.6}, Time=@n14) {centerX=0.4, centerY=0.4, speed=1}, Time=@n16) {speedX=0.05, speedY=0.07}
  n18 = Panner(Coordinate=Rotator(Coordinate=TextureCoordinate() {uTiling=0.3, vTiling=0.3}, Time=@n14) {centerX=0.2, centerY=0.2, speed=1}, Time=@n16) {speedX=-0.01, speedY=0.06}
  n19 = LinearInterpolate(
   A=TextureSample(UVs=@n17) {samplerType=SAMPLERTYPE_LinearGrayscale, texture=/Game/Environments/Hexmap/HexClouds/T_HexmapFog_01_Height_GS}.R
   B=TextureSample(UVs=@n18) {samplerType=SAMPLERTYPE_LinearGrayscale, texture=/Game/Global/Textures/Default/Flowmaps/T_StormSwirlMask}.R
  ) {constAlpha=0.4}
  n20 = LinearInterpolate(A=@n7.RGB, B=VectorParameter() {parameterName=SmokeColor2, defaultValue=<rgba 0.026 0.0274 0.04 0>}.RGB, Alpha=@n19)
  n21 = Multiply(
   A=LinearInterpolate(
    A=TextureSample(UVs=@n17) {samplerType=SAMPLERTYPE_Normal, texture=/Game/Environments/Hexmap/HexClouds/T_HexmapFog_01_N}.RGB
    B=TextureSample(UVs=@n18) {samplerType=SAMPLERTYPE_Normal, texture=/Game/VisualThreatManager/StormVisuals/LargeClouds/T_StormSwirlNormal}.RGB
   ) {constAlpha=0.4}
   B=@n2.RGB
  )
  n22 = Constant() {r=0.7}
  n23 = Fresnel(Normal=Transform(None=MaterialFunctionCall(Normal=@n21, Flatness=@n22) {materialFunction=&FlattenNormal}.Result)) {baseReflectFraction=0, exponent=0.1}
  n24 = SphereMask(A=TextureCoordinate(), B=Constant() {r=0.5}) {attenuationRadius=0.3, hardnessPercent=0}
  n25 = ShadingPathSwitch(Default=Clamp(None=LinearInterpolate(Alpha=@n23) {constA=1.7, constB=-0.4}), Mobile=Multiply(A=Power(Base=@n24)) {constB=0.5}) {desc=Migrated from FeatureLevelSwitch}
  n26 = Multiply(A=@n7.RGB, B=Multiply(A=Multiply(A=@n25, B=@n19)) {constB=4})
  n27 = VectorParameter() {parameterName=EmissiveColor, defaultValue=<rgba 0.8921 0.306 1.25 0>}
  n28 = DotProduct(A=Transform(None=@n21), B=CameraVectorWS())
  n29 = Constant() {r=0.5}
  n30 = ShadingPathSwitch(Default=Multiply(A=@n28, B=@n28), Mobile=SphereMask(A=TextureCoordinate(), B=@n29, Radius=@n29) {hardnessPercent=0}) {desc=Migrated from FeatureLevelSwitch}
  n31 = Power(Base=@n30) {constExponent=4}
  n32 = Multiply(
   A=MaterialFunctionCall!(TimeMult=@n8.Offset) {materialFunction=/Game/Effects/Fort_Effects/Materials/_Functions/MF_Lightning_Flashes}.Result
   B=Add(A=@n31) {constB=0.05}
  )
  n33 = Clamp(None=LinearInterpolate(
    A=ScalarParameter() {defaultValue=1, parameterName=EdgeGlowMax}
    B=ScalarParameter() {defaultValue=-0.5, parameterName=EdgeGlowMin}
    Alpha=@n19
   ) {constA=1, constB=-0.3})
  n34 = Add(A=Add(A=Add(A=@n20, B=@n26), B=Multiply(A=@n27.RGB, B=@n32)), B=Multiply(A=@n27.RGB, B=Multiply(A=@n33, B=@n31)))
  MaterialFunctionCall(Emissive=Add(A=ShadingPathSwitch(Default=@n4, Mobile=@n5) {desc=Migrated from FeatureLevelSwitch}, B=Multiply(A=@n6.RGB, B=@n34))) {materialFunction=&MF_FX_EmissiveAdjustments}
  Multiply()
  n35 = Multiply(A=Clamp(None=LinearInterpolate(Alpha=@n19) {constA=-0.1}), B=@n30)
  n36 = ScalarParameter() {defaultValue=4, parameterName=OpacityMult}
  n37 = Multiply(
   A=Clamp(None=Multiply(A=@n35, B=@n36) {constB=4}) {maxDefault=2}
   B=SphereMask(A=TextureCoordinate(), B=Constant() {r=0.5}) {attenuationRadius=0.48, hardnessPercent=90}
  ) {constA=1}
  n38 = MaterialFunctionCall("FadeSensitivity (optional)"=ScalarParameter() {parameterName="FadeSensitivity (optional) (S)"}) {materialFunction=&FadeBasedOnViewAngle}
  Multiply(
   A=LinearInterpolate(A=@n37, B=Multiply(A=@n37, B=@n38.Result), Alpha=ScalarParameter() {parameterName=UseViewAngleFade})
   B=ParticleColor().A
  )
  VertexNormalWS()
 ]
}
