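# Skinned-mesh material in the melee weapon cosmetics folder, adding an animated rainbow sparkle finish over a gun master base material.
# A panning rainbow gradient, distorted by a swirl flowmap, tints base color and feeds layered sparkle noise that drives emissive and hue shift.
# Sparkles use pivot-centered screen coordinates corrected for view size and FOV.
# Fresnel plus Sobel edges from the normal map masks the metallic, zero-roughness layer blend.
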Material /BRCosmetics/Weapons/FORT_Melee/Materials/Glider_Alteration
graph {
 properties {
  bUseMaterialAttributes=true
  bUsedWithMeshDeformer=true
  bUsedWithSkeletalMesh=true
  numCustomizedUVs=5
  pixelDepthOffsetMode=PDOM_Legacy
  usedShadingModels=Default Lit
 }
 nodes [
  n1 = MaterialFunctionCall(XY=TextureCoordinate()) {materialFunction=&Swizzle}
  n2 = ComponentMask(None=TextureSample(UVs=Panner() {speedY=-0.0125}) {samplerType=SAMPLERTYPE_LinearColor, texture=/Game/Global/Textures/Default/Flowmaps/T_SwirlFlow_01}.RGB) {g=true, r=true}
  n3 = TextureSample(UVs=Add(A=Panner(Coordinate=@n1.YX) {speedX=0.1}, B=Multiply(A=@n2) {constB=0.075})) {texture=/Game/Global/Textures/Default/Masks_and_Gradients/FortniteRainbow}
  n4 = ScalarParameter() {parameterName=Sobel mip}
  n5 = TextureObjectParameter() {parameterName=Normals, samplerType=SAMPLERTYPE_Normal, texture=/Game/Weapons/FORT_Shotguns/Textures/T_HammerShotGun_N}
  n6 = MaterialFunctionCall(Float2=TextureProperty(Tex=@n5)) {materialFunction=&BreakOutFloat2Components}
  n7 = Multiply(A=Divide(B=Max(A=@n6.R, B=@n6.R)), B=ScalarParameter() {defaultValue=0.25, parameterName=Sobel offset})
  n8 = StaticBool() {value=true}
  n9 = MaterialFunctionCall!(
   Base=MaterialFunctionCall(Power=ScalarParameter() {defaultValue=3.2002, parameterName=Fresnel Power}) {materialFunction=&Fresnel_Function}.Result
   Add=Power(
    Base=MaterialFunctionCall!(MipBias=@n4, OffsetAmount=@n7, Texture=@n5, Desaturate?=@n8, DesatColor, Modulate Results Before Clamp) {materialFunction=/Game/Weapons/WeaponSkins/MaterialFunction/SobelFilter}.Result
    Exp=ScalarParameter() {defaultValue=6, parameterName=Sobel Result Power}
   )
  ) {materialFunction=/Engine/Functions/Engine_MaterialFunctions02/Math/MultiplyAdd}
  Desaturation(None=Multiply(A=@n3.RGB, B=@n9.Result), Fraction=Constant() {r=3})
  n10 = GetMaterialAttributes(None=MaterialFunctionCall!(UV 1 coordinates, MuzzleFalloff) {materialFunction=/Game/Packages/Fortnite_UniqueMaterials_1/Weapons/M_GunMasterShaderFunction}.Material Result) {attributeGetTypes [
    #1
    #2
    #3
    #4
    #5
    #6
   ]}
  n11 = Add(A=@n10.Base Color, B=Multiply(A=@n3.RGB, B=ScalarParameter() {defaultValue=3, parameterName=RainbowIntensity}))
  n12 = MaterialFunctionCall() {materialFunction=&ComponentPivotLocation}
  n13 = ViewSize()
  n14 = Divide(B=Divide(A=ComponentMask(None=@n13) {r=true}) {constB=2048})
  n15 = MaterialFunctionCall() {materialFunction=/Engine/Functions/Engine_MaterialFunctions02/Utility/FOV}
  n16 = Multiply(
   A=ScalarParameter() {defaultValue=-0.2224, parameterName=Sparkle Scale}
   B=Multiply(
    A=MaterialFunctionCall!(Center Point World Position=@n12.Result) {materialFunction=/Game/Packages/Fortnite_UniqueMaterials_1/Effects/WorldAlignedScreenCoordinates}.Result
    B=Multiply(A=@n14, B=@n15."FOV ")
   )
  )
  n17 = TextureSample(UVs=Multiply(A=@n16) {constB=1.1}) {samplerType=SAMPLERTYPE_LinearColor, texture=/Game/Gadgets/Assets/Slow_Field/Textures/T_Slowfield_SparkleNoise_01}
  n18 = TextureSample(UVs=Panner(Coordinate=@n16) {speedX=0.1, speedY=0.1}) {samplerType=SAMPLERTYPE_LinearColor, texture=&T_Slowfield_SparkleNoise_01}
  n19 = Panner(Coordinate=Multiply(A=@n16) {constB=0.75}) {speedX=0.1, speedY=0.1}
  n20 = Multiply(
   A=Multiply(
    A=Multiply(A=@n17.R, B=@n18.G)
    B=TextureSample(UVs=@n19) {samplerType=SAMPLERTYPE_LinearColor, texture=&T_Slowfield_SparkleNoise_01}.G
   )
   B=ScalarParameter() {defaultValue=25.3628, parameterName=Sparkle Brightness}
  ) {constB=80}
  n21 = Multiply(A=@n11, B=@n20)
  MaterialFunctionCall(Texture=@n21) {materialFunction=&HueShift}
  n22 = MaterialFunctionCall(Hue Shift Percentage=@n20, Texture=@n11) {materialFunction=&HueShift}
  n23 = Constant() {r=1}
  n24 = Constant() {r=0.5}
  n25 = Constant()
  MaterialFunctionCall(In=MaterialFunctionCall!(
    Base Material=@n10.MaterialAttributes
    Top Material=MakeMaterialAttributes!(
     BaseColor=@n22.Result
     Metallic=@n23
     Specular=@n24
     Roughness=@n25
     Anisotropy
     EmissiveColor=@n21
     Opacity
     OpacityMask
     Normal=@n10.Normal
     Tangent
     WorldPositionOffset
     SubsurfaceColor
     ClearCoat
     ClearCoatRoughness
     AmbientOcclusion
     Refraction
     CustomizedUVs_0
     CustomizedUVs_1
     CustomizedUVs_2
     CustomizedUVs_3
     CustomizedUVs_4
     CustomizedUVs_5
     CustomizedUVs_6
     CustomizedUVs_7
     PixelDepthOffset
     ShadingModel
     Displacement
    )
    Alpha=@n9.Result
   ) {materialFunction=/Engine/Functions/MaterialLayerFunctions/MatLayerBlend_Standard}.Blended Material) {materialFunction=&MF_TechArtOverrides_accessories}
  TextureObjectParameter() {parameterName=Diffuse, texture=/Game/Weapons/FORT_Shotguns/Textures/T_HammerShotGun_D}
  TextureObject() {texture=&T_HammerShotGun_D}
  TextureSampleParameter2D() {group=Normals, parameterName=Normals, samplerType=SAMPLERTYPE_Normal, texture=&T_HammerShotGun_N}
  TextureSample() {texture=/Game/Characters/Fortress_CharacterShader/Textures/RainbowEffect}
 ]
}
