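# Physics asset for the RockerPunk pickaxe melee weapon mesh.
# It holds a single capsule body on the b_AR_Root bone, offset and rotated to roughly enclose the weapon.
# There are no constraints, so it gives the mesh one simple collision shape with no jointed parts.
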PhysicsAsset /BRCosmetics/Weapons/FORT_Melee/Meshes/PickAxe_RockerPunk_PhysicsAsset
bodies {b_AR_Root [{shapeType=Capsule, length=73.2306, radius=35.4028, center=<13.8341 0.3663 -7.1612>, rotation=<rot 0.503 -90 -0.9456>}]}
constraints []
