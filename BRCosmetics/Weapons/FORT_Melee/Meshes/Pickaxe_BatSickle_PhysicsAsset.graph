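# Physics asset for the Bat Sickle pickaxe melee weapon mesh, a Battle Royale cosmetic.
# It holds a single capsule body on the b_AR_Root bone, slightly offset and rotated to approximate the weapon's shape.
# There are no constraints, so the asset gives the weapon one rigid collision volume with no articulated parts.
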PhysicsAsset /BRCosmetics/Weapons/FORT_Melee/Meshes/Pickaxe_BatSickle_PhysicsAsset
bodies {b_AR_Root [{shapeType=Capsule, length=60.3769, radius=32.7091, center=<7.1932 -0.5906 -5.0594>, rotation=<rot 0.7268 -90 -2.0907>}]}
constraints []
