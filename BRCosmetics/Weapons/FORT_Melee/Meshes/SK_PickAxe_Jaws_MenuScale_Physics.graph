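# Physics asset for the menu-scale skeletal mesh of the Jaws pickaxe, a two-handed melee cosmetic.
# A long capsule covers the handle, with capsules on the dyn_ head and tail bones and a zero-length, effectively spherical one on the rope.
# All three dyn_ bodies constrain directly to the handle rather than chaining, with small angular limit values that are tightest on the rope.
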PhysicsAsset /BRCosmetics/Weapons/FORT_Melee/Meshes/SK_PickAxe_Jaws_MenuScale_Physics
bodies {
 Melee_TwoHanded_Handle [{shapeType=Capsule, length=290.5784, radius=7.1925, center=<-0.1686 -4.6529 56.0165>, rotation=<rot -0.8959 -90 0.009>}]
 dyn_jaws_head [{shapeType=Capsule, length=44.3277, radius=25.645, center=<-0.9503 47.1981 -3.6378>, rotation=<rot 0 1.9048 -77.4653>}]
 dyn_jaws_rope [{shapeType=Capsule, length=0, radius=8, center=<1.3139 -0.9389 2.7483>, rotation=<rot 1.8 -90 1.6>}]
 dyn_jaws_tail [{shapeType=Capsule, length=40.6619, radius=26.0122, center=<-0.2294 -43.4447 -7.0994>, rotation=<rot -0.0397 -179.7748 -79.5766>}]
}
constraints [
 {bone1Name=dyn_jaws_rope, bone2Name=Melee_TwoHanded_Handle, swing1LimitDegrees=0, swing2LimitDegrees=0.1, twistLimitDegrees=0.25}
 {bone1Name=dyn_jaws_tail, bone2Name=Melee_TwoHanded_Handle, swing1LimitDegrees=10, swing2LimitDegrees=0, twistLimitDegrees=10}
 {bone1Name=dyn_jaws_head, bone2Name=Melee_TwoHanded_Handle, swing1LimitDegrees=10, swing2LimitDegrees=0, twistLimitDegrees=10}
]
