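# Physics asset for the Jaws pickaxe skeletal mesh, a two-handed melee harvesting tool with simulated dangling parts.
# Capsule bodies cover the long handle, the head and tail jaw pieces, and a small rope segment.
# The head and tail each connect to the handle with one limited swing axis and limited twist.
# The rope connects to the handle with wider swing on both axes and locked twist.
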PhysicsAsset /BRCosmetics/Weapons/FORT_Melee/Meshes/SK_PickAxe_Jaws_PhysicsAsset
bodies {
 Melee_TwoHanded_Handle [{shapeType=Capsule, length=149.703, radius=4.2209, center=<-0.145 -2.14 30.6254>, rotation=<rot 0.2772 -90 -0.0215>}]
 dyn_jaws_head [{shapeType=Capsule, length=20.8531, radius=13.5603, center=<0.7956 10.5645 -0.5944>, rotation=<rot 0.4646 5.32 -84.2302>}]
 dyn_jaws_rope [{shapeType=Capsule, length=12, radius=4, center=<0.3872 0 0>}]
 dyn_jaws_tail [{shapeType=Capsule, length=18.7291, radius=12.4037, center=<0.012 -8.6365 -0.8802>, rotation=<rot -0.0101 -0.0378 -99.9234>}]
}
constraints [
 {bone1Name=Melee_TwoHanded_Handle, bone2Name=dyn_jaws_tail, swing1LimitDegrees=30, swing2LimitDegrees=0, twistLimitDegrees=30}
 {bone1Name=Melee_TwoHanded_Handle, bone2Name=dyn_jaws_head, swing1LimitDegrees=30, swing2LimitDegrees=0, twistLimitDegrees=30}
 {
  bone1Name=dyn_jaws_rope
  bone2Name=Melee_TwoHanded_Handle
  swing1LimitDegrees=80
  swing2LimitDegrees=80
  twistLimitDegrees=45
  twistMotion=Locked
 }
]
