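# Physics asset for the Pickaxe 10 melee weapon skeletal mesh in the BRCosmetics weapons content.
# A single capsule body on the b_AR_Root bone stands in for the weapon's shape, slightly offset from the bone and turned about a quarter turn.
# It has no constraints, so the asset defines no jointed or articulated parts.
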PhysicsAsset /BRCosmetics/Weapons/FORT_Melee/Meshes/SK_Pickaxe_10_Pickaxe_10_PhysicsAsset
bodies {b_AR_Root [{shapeType=Capsule, length=64.7105, radius=36.129, center=<6.1629 -0.3947 1.0211>, rotation=<rot 0.0964 -90 0.3791>}]}
constraints []
