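# Physics asset for the skeletal mesh SK_Pickaxe_11, a cosmetic melee pickaxe in the BRCosmetics weapon set.
# It has a single body on the b_AR_Root bone, made of one capsule turned sideways and slightly offset from the bone origin.
# There are no constraints, so the pickaxe collides and simulates as one rigid piece.
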PhysicsAsset /BRCosmetics/Weapons/FORT_Melee/Meshes/SK_Pickaxe_11_PhysicsAsset
bodies {b_AR_Root [{shapeType=Capsule, length=75.2647, radius=39.4009, center=<7.7841 -0.1993 -1.4633>, rotation=<rot 0.2167 -90 -2.2793>}]}
constraints []
