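# Physics asset for the SK_Pickaxe_12 skeletal mesh, a cosmetic melee pickaxe under BRCosmetics.
# It approximates the weapon with a single capsule body on the b_AR_Root bone, offset slightly from the bone origin and rotated about a quarter turn.
# There are no constraints, so the asset defines one rigid collision shape with no articulated parts.
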PhysicsAsset /BRCosmetics/Weapons/FORT_Melee/Meshes/SK_Pickaxe_12_PhysicsAsset
bodies {b_AR_Root [{shapeType=Capsule, length=64.8458, radius=20.4158, center=<7.4942 -1.1378 -2.6318>, rotation=<rot 0.2323 -90 -0.5234>}]}
constraints []
