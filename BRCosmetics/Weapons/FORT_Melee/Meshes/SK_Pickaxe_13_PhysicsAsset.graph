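# Physics asset for the SK_Pickaxe_13 two-handed melee pickaxe skeletal mesh in the BRCosmetics weapons content.
# It approximates the weapon with two capsule bodies: a slim one on the Melee_TwoHanded_Handle bone and a much bulkier one on b_AR_Root.
# A single constraint links the two bodies, with matching moderate limits on both swing axes and twist.
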PhysicsAsset /BRCosmetics/Weapons/FORT_Melee/Meshes/SK_Pickaxe_13_PhysicsAsset
bodies {
 Melee_TwoHanded_Handle [{shapeType=Capsule, length=26.472, radius=4.1693, center=<0.242 -0.9568 -18.4137>, rotation=<rot -0.004 -90 -3.5417>}]
 b_AR_Root [{shapeType=Capsule, length=44.2611, radius=33.7644, center=<7.2749 2.8617 32.0136>, rotation=<rot 1.6141 -90 -0.4503>}]
}
constraints [{bone1Name=Melee_TwoHanded_Handle, bone2Name=b_AR_Root, swing1LimitDegrees=45, swing2LimitDegrees=45, twistLimitDegrees=45}]
