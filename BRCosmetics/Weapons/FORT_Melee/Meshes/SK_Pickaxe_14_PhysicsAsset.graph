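# Physics asset for the skeletal mesh SK_Pickaxe_14, a Battle Royale cosmetic melee pickaxe.
# It holds a single capsule body on the b_AR_Root bone, offset and rotated to serve as a rough collision volume for the pickaxe.
# It has no constraints, so it simulates as one rigid piece with no articulated parts.
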PhysicsAsset /BRCosmetics/Weapons/FORT_Melee/Meshes/SK_Pickaxe_14_PhysicsAsset
bodies {b_AR_Root [{shapeType=Capsule, length=65.9228, radius=19.3127, center=<6.2896 -5.108 -6.4046>, rotation=<rot -0.5225 -90 -0.2787>}]}
constraints []
