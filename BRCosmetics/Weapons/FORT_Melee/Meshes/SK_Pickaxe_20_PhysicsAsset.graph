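# Physics asset for the SK_Pickaxe_20 skeletal mesh, a cosmetic melee pickaxe weapon, giving it simple collision for physics.
# It holds a single body on the b_AR_Root bone, shaped as one offset, rotated capsule.
# It has no constraints, so the pickaxe is represented as one rigid piece.
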PhysicsAsset /BRCosmetics/Weapons/FORT_Melee/Meshes/SK_Pickaxe_20_PhysicsAsset
bodies {b_AR_Root [{shapeType=Capsule, length=68.4028, radius=36.5977, center=<6.3464 -0.5786 -7.0178>, rotation=<rot 0.0058 -90 2.6165>}]}
constraints []
