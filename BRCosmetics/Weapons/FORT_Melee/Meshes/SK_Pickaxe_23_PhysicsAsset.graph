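# Physics asset for the SK_Pickaxe_23 melee weapon cosmetic mesh, giving the pickaxe a simplified collision shape.
# It holds a single body on the LOW_CircuitBreaker002 bone, made of one large capsule offset from the bone origin and turned roughly ninety degrees.
# It has no constraints, so there are no articulated parts to simulate.
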PhysicsAsset /BRCosmetics/Weapons/FORT_Melee/Meshes/SK_Pickaxe_23_PhysicsAsset
bodies {LOW_CircuitBreaker002 [{shapeType=Capsule, length=325.126, radius=186.2853, center=<-16.9819 93.873 -0.233>, rotation=<rot 0 2.4397 -89.9474>}]}
constraints []
