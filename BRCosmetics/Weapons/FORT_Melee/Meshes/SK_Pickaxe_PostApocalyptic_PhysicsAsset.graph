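# Physics asset for the Post-Apocalyptic pickaxe skeletal mesh, a two-handed melee harvesting tool cosmetic.
# It holds a single body on the Melee_TwoHanded_Handle bone, shaped as one large capsule that roughly encloses the weapon.
# With no constraints, it provides a simple collision volume rather than an articulated ragdoll.
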PhysicsAsset /BRCosmetics/Weapons/FORT_Melee/Meshes/SK_Pickaxe_PostApocalyptic_PhysicsAsset
bodies {Melee_TwoHanded_Handle [{shapeType=Capsule, length=62.3935, radius=31.1273, center=<-0.3256 1.64 29.471>, rotation=<rot -1.323 -90 0.0946>}]}
constraints []
