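# Physics asset for the Skiboot pickaxe skeletal mesh, a cosmetic melee weapon.
# It holds a single body on the Melee_TwoHanded_Handle bone, shaped as one offset, rotated box.
# It defines no constraints, so there are no jointed or independently simulated parts.
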PhysicsAsset /BRCosmetics/Weapons/FORT_Melee/Meshes/SK_Pickaxe_Skiboot_PhysicsAsset
bodies {Melee_TwoHanded_Handle [{shapeType=Box, extentX=85.9746, extentY=16.8899, extentZ=135.6987, center=<1.0578 4.3888 23.566>, rotation=<rot 1.694 -90 1.8678>}]}
constraints []
