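# Physics asset for the skull pickaxe melee weapon skeletal mesh, providing its simplified physics collision shapes.
# Two capsule bodies cover it: a slim one on Melee_TwoHanded_Handle and a much larger one on b_AR_Root.
# A single constraint joins the two bones with both swings and twist set to free, so the stored angular limits do not restrict rotation.
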PhysicsAsset /BRCosmetics/Weapons/FORT_Melee/Meshes/SK_Pickaxe_Skull_PhysicsAsset
bodies {
 Melee_TwoHanded_Handle [{shapeType=Capsule, length=24.7066, radius=5.2469, center=<0.6873 -0.571 -10.234>, rotation=<rot 0.7216 -90 -0.7972>}]
 b_AR_Root [{shapeType=Capsule, length=35.118, radius=30.9863, center=<8.9626 2.2716 33.46>, rotation=<rot 0 -100.1829 -178.7949>}]
}
constraints [{
  bone1Name=Melee_TwoHanded_Handle
  bone2Name=b_AR_Root
  swing1LimitDegrees=45
  swing1Motion=Free
  swing2LimitDegrees=45
  swing2Motion=Free
  twistLimitDegrees=45
  twistMotion=Free
 }]
