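# Physics asset for the Smiley pickaxe skeletal mesh, a cosmetic melee weapon.
# It holds a single body on the b_AR_Root bone, shaped as one capsule offset and rotated from the bone to approximate the pickaxe.
# It has no constraints, so the weapon simulates and collides as one rigid piece.
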PhysicsAsset /BRCosmetics/Weapons/FORT_Melee/Meshes/SK_Pickaxe_Smiley_PhysicsAsset
bodies {b_AR_Root [{shapeType=Capsule, length=61.2336, radius=34.0225, center=<5.6421 0.1048 -13.2738>, rotation=<rot -0.6901 -90 2.4592>}]}
constraints []
