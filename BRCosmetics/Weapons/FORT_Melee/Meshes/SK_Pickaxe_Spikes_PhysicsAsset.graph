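# Physics asset for the Spikes pickaxe skeletal mesh, a two-handed melee weapon cosmetic.
# It holds a single capsule body on the Melee_TwoHanded_Handle bone, offset along the handle and rotated to lie across the weapon.
# No constraints exist, so the whole pickaxe is represented as one rigid collision shape.
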PhysicsAsset /BRCosmetics/Weapons/FORT_Melee/Meshes/SK_Pickaxe_Spikes_PhysicsAsset
bodies {Melee_TwoHanded_Handle [{shapeType=Capsule, length=57.2573, radius=34.2978, center=<0.1711 -1.3275 22.3119>, rotation=<rot -0.7447 -90 -0.1044>}]}
constraints []
