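# Physics asset for the skeletal mesh of the Warthog pickaxe, a two-handed melee cosmetic weapon.
# It holds a single body on the Melee_TwoHanded_Handle bone, shaped as one large capsule that is offset and rotated relative to that bone.
# It defines no constraints, so the asset provides one simple collision volume with no jointed parts.
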PhysicsAsset /BRCosmetics/Weapons/FORT_Melee/Meshes/SK_Pickaxe_Warthog_PhysicsAsset
bodies {Melee_TwoHanded_Handle [{shapeType=Capsule, length=61.0377, radius=31.8939, center=<0.0313 2.2156 25.4684>, rotation=<rot -4.6395 -90 -0.0435>}]}
constraints []
